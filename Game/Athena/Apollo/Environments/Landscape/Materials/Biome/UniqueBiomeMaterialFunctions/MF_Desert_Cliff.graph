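# Landscape biome material function for desert cliff rock, outputting material attributes plus a height from diffuse alpha.
# World-aligned cliff diffuse, spec and normal textures use height-refined projection transitions, with base color overlay-tinted by an alpha-driven HeightLerp.
# The default shading path lightly blends in world-aligned rock detail, which Forward and Mobile skip.
# Loose parameters mirror the hardcoded constants, and an altitude-based darkening chain is left unconnected to the outputs.
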MaterialFunction /Game/Athena/Apollo/Environments/Landscape/Materials/Biome/UniqueBiomeMaterialFunctions/MF_Desert_Cliff
graph {nodes [
  Constant3Vector() {constant=<rgba 2100 2100 1600 1>}
  FunctionInput() {inputName=Slope, inputType=FunctionInput_Scalar, sortPriority=4}
  n1 = MaterialFunctionCall!(
   Diffuse Texture Object=TextureObject() {texture=/Game/Environments/Artemis/Sets/Terrain/Textures/T_Artemis_Desert_Cliff_02_D}
   Spec Texture Object=TextureObject() {samplerType=SAMPLERTYPE_Masks, texture=/Game/Environments/Artemis/Sets/Terrain/Textures/T_Artemis_Desert_Cliff_02_S}
   TextureSize=Constant() {r=1813.2452}
   --------------------------
   Use seperate settings for Z projection
   Use seperate texture for Z projection
   Texture Size Z projection
   Z Plane Diffuse Texture Object
   Z Plane Spec Texture Object
   "--------------------------      "
   ProjectionTransitionContrast=Constant() {r=0.5}
   Use Texture Alpha Heightmap to Refine the Transistion=StaticBool() {value=true}
   HeightLerpTransitionContrast=Constant() {r=1}
   PickTheDiffuseHeightChannel
   TheDiffuseHeightChannel
   "--------------------------       "
   NormalTextureObject=TextureObject() {samplerType=SAMPLERTYPE_Normal, texture=/Game/Environments/Artemis/Sets/Terrain/Textures/T_Artemis_Desert_Cliff_02_N}
   World Space Normals=StaticBool()
   "--------------------------  "
   Export Float 4 Diffuse=StaticBool() {value=true}
   Export Float 4 Spec
   "--------------------------     "
   WorldPosition=WorldPosition() {worldPositionShaderOffset=WPT_ExcludeAllShaderOffsets}.XYZ
   World Space Normal
  ) {materialFunction=/Engine/Functions/Engine_MaterialFunctions02/Texturing/WorldAlignedTextures_Complex}
  n2 = Constant() {r=0.6406}
  n3 = Constant() {r=0.5573}
  n4 = Constant() {r=0.5}
  n5 = ComponentMask(None=@n1.XY Diffuse Texture) {a=true}
  n6 = Constant() {r=3}
  n7 = ComponentMask(None=@n1.XY Spec) {r=true}
  n8 = ComponentMask(None=@n1.XY Spec) {b=true}
  n9 = SetMaterialAttributes!(
   MaterialAttributes
   BaseColor=MaterialFunctionCall(
    Base=ComponentMask(None=@n1.XY Diffuse Texture) {b=true, g=true, r=true}
    Blend=MaterialFunctionCall(A=@n2, B=@n3, Transition Phase=@n4, Height Texture=@n5, Contrast=@n6) {materialFunction=&HeightLerp}.Results
   ) {materialFunction=&Blend_Overlay}.Result
   Specular=Multiply(A=Constant() {r=1.1414}, B=@n7)
   Roughness=@n8
   Normal=@n1.Normal XY Texture
  ) {attributeSetTypes [
    #1
    #2
    #3
    #4
   ]}
  n10 = GetMaterialAttributes(None=@n9) {attributeGetTypes [
    #1
    #2
    #3
    #4
   ]}
  n11 = ComponentMask(None=MaterialFunctionCall(
    TextureObject=TextureObject() {texture=/Game/Environments/Apollo/Rocks/Natural_Rocks/Textures/T_RockDetails_02D}
    TextureSize=Constant() {r=1565.0168}
   ) {materialFunction=/Game/Weapons/WeaponSkins/Wraps/MaterialFunctions/WorldAlignedTexture}.XYZ Texture) {r=true}
  n12 = Multiply(A=Constant() {r=1.8102}, B=@n11)
  n13 = LinearInterpolate(A=Constant() {r=-2.505}, B=Constant() {r=1.7983}, Alpha=@n11)
  n14 = SetMaterialAttributes(
   MaterialAttributes=@n10.MaterialAttributes
   BaseColor=Saturate(None=MaterialFunctionCall(Base=@n10.BaseColor, Blend=@n12) {materialFunction=&Blend_Overlay}.Result)
   Specular=Saturate(None=Multiply(A=@n11, B=@n10.Specular))
   Roughness=Saturate(None=Multiply(A=@n10.Roughness, B=@n13))
  ) {attributeSetTypes [
    #1
    #2
    #3
    #4
   ]}
  n15 = Constant() {r=0.1}
  FunctionOutput(None=ShadingPathSwitch(Default=BlendMaterialAttributes(A=@n10.MaterialAttributes, B=@n14, Alpha=@n15), Forward=@n9, Mobile=@n9)) {outputName=Material, sortPriority=-1}
  n16 = ComponentMask(None=@n1.XY Diffuse Texture) {a=true}
  FunctionOutput(None=NamedRerouteDeclaration(None=Multiply(A=@n16)) {name=OutputHeight, nodeColor=<rgba 1 0 1 1>}) {outputName=Height}
  n17 = SetMaterialAttributes(BaseColor=ComponentMask(None=@n1.XY Diffuse Texture) {b=true, g=true, r=true}, Specular=@n7, Roughness=@n8, Normal=@n1.Normal XY Texture) {attributeSetTypes [
    #1
    #2
    #3
    #4
   ]}
  NamedRerouteDeclaration(None=@n17) {name=MediumRock-Material, nodeColor=<rgba 1 0 1 1>}
  NamedRerouteDeclaration(None=OneMinus(None=@n16)) {name=MediumRock-Height, nodeColor=<rgba 1 0 1 1>}
  ScalarParameter() {defaultValue=0.5, group=Side, parameterName=ColorViaAlpha, sortPriority=2}
  ScalarParameter() {defaultValue=3, group=Side, parameterName=ColorViaConrast, sortPriority=2}
  ScalarParameter() {defaultValue=0.5, parameterName=DC-MR-TC}
  ScalarParameter() {defaultValue=-2.505, group=Detail Texture, parameterName=Detail R Min, sortPriority=5}
  ScalarParameter() {defaultValue=1.7983, group=Detail Texture, parameterName=Detail R Max, sortPriority=6}
  ScalarParameter() {defaultValue=1.8102, group=Detail Texture, parameterName=Details Tone}
  ScalarParameter() {defaultValue=1565.0168, group=Detail Texture, parameterName=DetailTile}
  ScalarParameter() {defaultValue=0.1, group=Detail Texture, parameterName=Details Visibility}
  ScalarParameter() {defaultValue=1, parameterName=DC-MR-HLTC}
  ScalarParameter() {defaultValue=1.1414, group=Side, parameterName=Side Spec Tweak}
  n18 = Reroute()
  n19 = MaterialFunctionCall!(DarknessAmount) {
   desc=Cliffs get darker with increasing altitude
   materialFunction=/Game/Athena/Apollo/Environments/Landscape/Materials/Biome/CommonMaterials/MF_HeightBasedDarkenCliff
  }
  n20 = Add(A=@n19.WorldZ, B=Constant() {r=3950})
  n21 = Constant() {r=250}
  n22 = Max(A=Saturate(None=Divide(A=@n20, B=@n21)), B=Constant() {r=0.5})
  n23 = GetMaterialAttributes(None=@n18) {attributeGetTypes [#1, #3]}
  SetMaterialAttributes!(
   MaterialAttributes=@n18
   BaseColor=Multiply(A=@n22, B=Multiply(A=@n19.Result, B=@n23.BaseColor))
   Roughness=Multiply(A=@n22, B=@n23.Roughness)
  ) {attributeSetTypes [#1, #3]}
  VectorParameter() {group=Side, parameterName=ColorTone B, sortPriority=1, defaultValue=<rgba 0.6406 0.6406 0.6406 0>}
  VectorParameter() {group=Rock Large, parameterName=DC-MedRockScale, defaultValue=<rgba 1813.2452 1813.2452 1813.2452 1>}
  VectorParameter() {group=Side, parameterName=ColorTone A, sortPriority=0, defaultValue=<rgba 0.5573 0.5573 0.5573 0>}
 ]}
