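# Material function for the Artemis landscape that turns a sampled scalar into a binary chrome physical material mask.
# An If node compares the SampleValue input against the ChromeTakeOverThreshold scalar parameter.
# It outputs one at or above the threshold and zero below it, giving a hard cutoff with no blending.
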MaterialFunction /Game/Athena/Apollo/Environments/Landscape/Materials/MaterialFunctions/MF_ArtemisChromePhysicalMaterialMask
graph {nodes [
  n1 = Constant() {r=1}
  FunctionOutput(None=If(
    A=FunctionInput() {inputName=SampleValue, inputType=FunctionInput_Scalar}
    B=ScalarParameter() {defaultValue=0.5, parameterName=ChromeTakeOverThreshold}
    "A > B"=@n1
    "A == B"=@n1
    "A < B"=Constant()
   ) {constB=0.7})
 ]}
