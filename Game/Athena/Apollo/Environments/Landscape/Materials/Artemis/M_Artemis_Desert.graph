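# Desert biome landscape material for the Artemis terrain, writing to and sampling from a runtime virtual texture.
# Common layers are blended with painted sand, rock and scree (BiomeA-C), then slope-blended cliffs and BiomeD-masked LackLuster (chrome) before the RVT write.
# The final surface samples the RVT, or uses simpler mobile layers, then adds snow, LackLuster, fire and debug passes.
# Layer, slope and height masks also drive landscape grass and physical materials.
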Material /Game/Athena/Apollo/Environments/Landscape/Materials/Artemis/M_Artemis_Desert
graph {
 properties {bTangentSpaceNormal=false, bUseMaterialAttributes=true, pixelDepthOffsetMode=PDOM_Legacy, usedShadingModels=Default Lit}
 nodes [
  Constant() {r=1}
  Constant() {r=3}
  Constant() {r=1.3}
  Constant() {r=-0.4}
  Constant() {r=3.5}
  Constant() {r=-0.3}
  Constant() {r=8192}
  n1 = NamedRerouteUsage()
  MaterialExpressionSubtract_1 = Subtract(
   A=MaterialFunctionCall!(In=@n1) {materialFunction=/Game/Athena/Apollo/Environments/Landscape/Materials/Biome/CommonMaterials/MF_LS_GrassSwitch}.Result
   B=LandscapeLayerSample() {parameterName=BiomeD}
  )
  MaterialExpressionSubtract_22 = Subtract(
   A=MaterialFunctionCall(In=@n1) {materialFunction=&MF_LS_GrassSwitch}.Result
   B=LandscapeLayerSample() {parameterName=BiomeD}
  )
  MaterialExpressionSubtract_4 = Subtract(
   A=MaterialFunctionCall() {materialFunction=&MF_LS_GrassSwitch}.Result
   B=LandscapeLayerSample() {parameterName=BiomeD}
  )
  MaterialExpressionMaterialFunctionCall_19 = MaterialFunctionCall() {materialFunction=&MF_LS_GrassSwitch}
  MaterialExpressionSubtract_7 = Subtract(
   A=MaterialFunctionCall() {materialFunction=&MF_LS_GrassSwitch}.Result
   B=LandscapeLayerSample() {parameterName=BiomeD}
  )
  MaterialExpressionSubtract_15 = Subtract(
   A=MaterialFunctionCall() {materialFunction=&MF_LS_GrassSwitch}.Result
   B=LandscapeLayerSample() {parameterName=BiomeD}
  )
  MaterialExpressionSubtract_19 = Subtract(
   A=MaterialFunctionCall() {materialFunction=&MF_LS_GrassSwitch}.Result
   B=LandscapeLayerSample() {parameterName=BiomeD}
  )
  MaterialExpressionMaterialFunctionCall_23 = MaterialFunctionCall(In=NamedRerouteUsage()) {materialFunction=&MF_LS_GrassSwitch}
  MaterialExpressionMaterialFunctionCall_24 = MaterialFunctionCall(In=NamedRerouteUsage()) {materialFunction=&MF_LS_GrassSwitch}
  MaterialExpressionSubtract_21 = Subtract(
   A=MaterialFunctionCall(In=NamedRerouteUsage()) {materialFunction=&MF_LS_GrassSwitch}.Result
   B=LandscapeLayerSample() {parameterName=BiomeD}
  )
  MaterialExpressionMaterialFunctionCall_26 = MaterialFunctionCall(In=NamedRerouteUsage()) {materialFunction=&MF_LS_GrassSwitch}
  MaterialExpressionSubtract_20 = Subtract(
   A=MaterialFunctionCall() {materialFunction=&MF_LS_GrassSwitch}.Result
   B=LandscapeLayerSample() {parameterName=BiomeD}
  )
  MaterialExpressionSubtract_18 = Subtract(
   A=MaterialFunctionCall() {materialFunction=&MF_LS_GrassSwitch}.Result
   B=LandscapeLayerSample() {parameterName=BiomeD}
  )
  LandscapeGrassOutput!(
   Grass_Standard=@MaterialExpressionSubtract_1
   Bush=@MaterialExpressionSubtract_22
   GrasslandFlowers=@MaterialExpressionSubtract_4
   Scree=@MaterialExpressionMaterialFunctionCall_19.Result
   Scree_Grass=@MaterialExpressionSubtract_7
   Heather=@MaterialExpressionSubtract_15
   ForestFloor=@MaterialExpressionSubtract_19
   Beach=@MaterialExpressionMaterialFunctionCall_23.Result
   Ocean=@MaterialExpressionMaterialFunctionCall_24.Result
   DesertPlants=@MaterialExpressionSubtract_21
   DesertScree=@MaterialExpressionMaterialFunctionCall_26.Result
   TropicalJungle=@MaterialExpressionSubtract_20
   FrontierPlants=@MaterialExpressionSubtract_18
  ) {
   desc="This must have the exact inputs in every biome material!\r\nThis data is baked into the level data when the level is\r\nsaved, so any changes to this require the affected level\r\nto be saved && submitted."
   grassTypes [
    Grass_Standard {
     grassType=/Game/Athena/Apollo/Environments/Landscape/SmallFoliage/DataAssets/LandscapeGrassType/LandscapeGrassType_StandardGrass
     input {
      expression=&self:MaterialExpressionSubtract_1
      inputName=None
      mask=0
      maskA=0
      maskB=0
      maskG=0
      maskR=0
      outputIndex=0
     }
    }
    Bush {
     grassType=/Game/Athena/Environments/Landscape/Foliage/Athena_Foliage_Bushes
     input {
      expression=&self:MaterialExpressionSubtract_22
      inputName=None
      mask=0
      maskA=0
      maskB=0
      maskG=0
      maskR=0
      outputIndex=0
     }
    }
    GrasslandFlowers {
     grassType=/Game/Athena/Apollo/Environments/Landscape/SmallFoliage/DataAssets/LandscapeGrassType/LandscapeGrassType_FarmFlowers
     input {
      expression=&self:MaterialExpressionSubtract_4
      inputName=None
      mask=0
      maskA=0
      maskB=0
      maskG=0
      maskR=0
      outputIndex=0
     }
    }
    Scree {
     grassType=/Game/Athena/Apollo/Environments/Landscape/SmallFoliage/DataAssets/LandscapeGrassType/LandscapeGrassType_Scree
     input {
      expression=&self:MaterialExpressionMaterialFunctionCall_19
      inputName=None
      mask=0
      maskA=0
      maskB=0
      maskG=0
      maskR=0
      outputIndex=0
     }
    }
    Scree_Grass {
     grassType=/Game/Athena/Apollo/Environments/Landscape/SmallFoliage/DataAssets/LandscapeGrassType/LandscapeGrassType_ScreeGrass
     input {
      expression=&self:MaterialExpressionSubtract_7
      inputName=None
      mask=0
      maskA=0
      maskB=0
      maskG=0
      maskR=0
      outputIndex=0
     }
    }
    Heather {
     grassType=/Game/Athena/Apollo/Environments/Landscape/SmallFoliage/DataAssets/LandscapeGrassType/LandscapeGrassType_Heather
     input {
      expression=&self:MaterialExpressionSubtract_15
      inputName=None
      mask=0
      maskA=0
      maskB=0
      maskG=0
      maskR=0
      outputIndex=0
     }
    }
    ForestFloor {
     grassType=/Game/Athena/Apollo/Environments/Landscape/SmallFoliage/DataAssets/LandscapeGrassType/LandscapeGrassType_ForestGrass
     input {
      expression=&self:MaterialExpressionSubtract_19
      inputName=None
      mask=0
      maskA=0
      maskB=0
      maskG=0
      maskR=0
      outputIndex=0
     }
    }
    Beach {
     grassType=/Game/Athena/Apollo/Environments/Landscape/SmallFoliage/DataAssets/LandscapeGrassType/LandscapeGrassType_Beach
     input {
      expression=&self:MaterialExpressionMaterialFunctionCall_23
      inputName=None
      mask=0
      maskA=0
      maskB=0
      maskG=0
      maskR=0
      outputIndex=0
     }
    }
    Ocean {
     grassType=/Game/Athena/Apollo/Environments/Landscape/SmallFoliage/DataAssets/LandscapeGrassType/LandscapeGrassType_Ocean
     input {
      expression=&self:MaterialExpressionMaterialFunctionCall_24
      inputName=None
      mask=0
      maskA=0
      maskB=0
      maskG=0
      maskR=0
      outputIndex=0
     }
    }
    DesertPlants {
     grassType=/Game/Athena/Apollo/Environments/Landscape/SmallFoliage/DataAssets/LandscapeGrassType/LandscapeGrassType_DesertPlants
     input {
      expression=&self:MaterialExpressionSubtract_21
      inputName=None
      mask=0
      maskA=0
      maskB=0
      maskG=0
      maskR=0
      outputIndex=0
     }
    }
    DesertScree {
     grassType=/Game/Athena/Apollo/Environments/Landscape/SmallFoliage/DataAssets/LandscapeGrassType/LandscapeGrassType_DesertScree
     input {
      expression=&self:MaterialExpressionMaterialFunctionCall_26
      inputName=None
      mask=0
      maskA=0
      maskB=0
      maskG=0
      maskR=0
      outputIndex=0
     }
    }
    TropicalJungle {
     grassType=/Game/Athena/Apollo/Environments/Landscape/SmallFoliage/DataAssets/LandscapeGrassType/LandscapeGrassType_Nightmare
     input {
      expression=&self:MaterialExpressionSubtract_20
      inputName=None
      mask=0
      maskA=0
      maskB=0
      maskG=0
      maskR=0
      outputIndex=0
     }
    }
    FrontierPlants {
     grassType=/Game/Athena/Apollo/Environments/Landscape/SmallFoliage/DataAssets/LandscapeGrassType/NewLandscapeGrassType_FrontierGrass
     input {
      expression=&self:MaterialExpressionSubtract_18
      inputName=None
      mask=0
      maskA=0
      maskB=0
      maskG=0
      maskR=0
      outputIndex=0
     }
    }
   ]
  }
  MaterialExpressionMaterialFunctionCall_30 = MaterialFunctionCall!(SampleValue=LandscapeLayerSample() {parameterName=BiomeD}) {materialFunction=/Game/Athena/Apollo/Environments/Landscape/Materials/MaterialFunctions/MF_ArtemisChromePhysicalMaterialMask}
  n2 = NamedRerouteUsage()
  n3 = OneMinus(None=@n2) {desc="White=Cliffs"}
  n4 = Max(
   A=Max(A=LandscapeLayerSample() {parameterName=BiomeA}, B=LandscapeLayerSample() {parameterName=BiomeB})
   B=LandscapeLayerSample() {parameterName=BiomeC}
  )
  n5 = Max(A=LandscapeLayerSample() {parameterName=Gravel}, B=LandscapeLayerSample() {parameterName=Mud})
  n6 = Constant() {r=2}
  n7 = Max(A=@n4, B=Saturate(None=Multiply(A=@n5, B=@n6)))
  n8 = Subtract(A=@n7, B=@n3) {desc="Remove Cliff from Gravel,Mud, Sand, Snow"}
  n9 = LandscapeLayerSample() {parameterName=Sand}
  n10 = Saturate(None=Multiply(A=LandscapeLayerSample() {parameterName=Asphalt}, B=Constant() {r=3}))
  MaterialExpressionSaturate_12 = Saturate(None=Subtract(A=Saturate(None=@n8), B=Max(A=@n9, B=@n10)) {desc="Remove Asphalt, Sand"}) {desc=Final Dirt mask}
  MaterialExpressionReroute_51 = Reroute(None=@n3) {desc=Final Stone mask}
  n11 = Subtract(A=@n2, B=Max(A=@n9, B=@n7)) {desc="Remove Gravel,Mud, Sand, Snow"}
  MaterialExpressionSaturate_14 = Saturate(None=Subtract(A=Saturate(None=@n11), B=@n10) {desc=Remove Asphalt}) {desc=Final Grass mask}
  MaterialExpressionReroute_61 = Reroute(None=@n10)
  n12 = Saturate(None=Subtract(A=@n9, B=@n3) {desc=Remove Cliff})
  MaterialExpressionReroute_42 = Reroute(None=Saturate(None=Subtract(A=@n12, B=@n10)))
  LandscapePhysicalMaterialOutput!(
   Chrome_Petrified_PhysicalMaterial=@MaterialExpressionMaterialFunctionCall_30.Result
   Dirt=@MaterialExpressionSaturate_12
   Stone=@MaterialExpressionReroute_51
   Grass=@MaterialExpressionSaturate_14
   Asphalt=@MaterialExpressionReroute_61
   Sand=@MaterialExpressionReroute_42
  ) {inputs [
    {
     physicalMaterial=/Game/TechArt/Seasonal/LackLuster/Physics/Chrome_Petrified_PhysicalMaterial
     input {
      expression=&self:MaterialExpressionMaterialFunctionCall_30
      inputName=None
      mask=0
      maskA=0
      maskB=0
      maskG=0
      maskR=0
      outputIndex=0
     }
    }
    {
     physicalMaterial=/Game/FortressPhysicalMaterials/Dirt
     input {
      expression=&self:MaterialExpressionSaturate_12
      inputName=None
      mask=0
      maskA=0
      maskB=0
      maskG=0
      maskR=0
      outputIndex=0
     }
    }
    {
     physicalMaterial=/Game/FortressPhysicalMaterials/Stone
     input {
      expression=&self:MaterialExpressionReroute_51
      inputName=None
      mask=0
      maskA=0
      maskB=0
      maskG=0
      maskR=0
      outputIndex=0
     }
    }
    {
     physicalMaterial=/Game/FortressPhysicalMaterials/Grass
     input {
      expression=&self:MaterialExpressionSaturate_14
      inputName=None
      mask=0
      maskA=0
      maskB=0
      maskG=0
      maskR=0
      outputIndex=0
     }
    }
    {
     physicalMaterial=/Game/FortressPhysicalMaterials/Asphalt
     input {
      expression=&self:MaterialExpressionReroute_61
      inputName=None
      mask=0
      maskA=0
      maskB=0
      maskG=0
      maskR=0
      outputIndex=0
     }
    }
    {
     physicalMaterial=/Game/FortressPhysicalMaterials/Sand
     input {
      expression=&self:MaterialExpressionReroute_42
      inputName=None
      mask=0
      maskA=0
      maskB=0
      maskG=0
      maskR=0
      outputIndex=0
     }
    }
   ]}
  n13 = LinearInterpolate(A=@MaterialExpressionSaturate_12, B=@MaterialExpressionReroute_51, Alpha=ScalarParameter() {parameterName=DEBUG-PhysMat1})
  n14 = ScalarParameter() {parameterName=DEBUG-PhysMat2}
  LinearInterpolate(
   A=LinearInterpolate(
    A=LinearInterpolate(A=@n13, B=@MaterialExpressionSaturate_14, Alpha=@n14)
    B=@MaterialExpressionReroute_61
    Alpha=ScalarParameter() {parameterName=DEBUG-PhysMat3}
   )
   B=@MaterialExpressionReroute_42
   Alpha=ScalarParameter() {defaultValue=1, parameterName=DEBUG-PhysMat4}
  )
  Max(A=@n4, B=@n10)
  n15 = WorldPosition() {worldPositionShaderOffset=WPT_ExcludeAllShaderOffsets}
  n16 = TextureSample(UVs=Divide(A=ComponentMask(None=@n15.XYZ) {g=true, r=true}, B=ScalarParameter() {defaultValue=3000, parameterName=DesertScree-NoiseTile}) {constB=8192}) {
   samplerSource=SSM_Wrap_WorldGroupSettings
   samplerType=SAMPLERTYPE_Grayscale
   texture=/Game/Athena/Apollo/Environments/Landscape/Materials/Biome/CommonMaterials/T_ApolloGrass_ColorMask
  }
  n17 = ScalarParameter() {defaultValue=1, parameterName=DesertScree-NoiseLerpA}
  Multiply(B=OneMinus(None=Power(Base=@n16.R, Exp=@n17)))
  n18 = MaterialFunctionCall!(AdditionalGrassBorderColorMask, PutSandBeforeGravel=StaticBool() {value=true}, UseOceanLevel) {materialFunction=/Game/Athena/Apollo/Environments/Landscape/Materials/Biome/CommonMaterials/MF_CommonLandscapeLayers}
  NamedRerouteDeclaration(None=@n18.CommonLayerMaskForGrassFoliage) {name=GM-CommonLayers, nodeColor=<rgba 0.5 0 1 1>}
  n19 = LandscapeLayerSample() {desc=DesertRock, parameterName=BiomeB}
  n20 = LandscapeLayerSample() {desc=DesertScree, parameterName=BiomeC}
  NamedRerouteDeclaration(None=Saturate(None=Subtract(A=@n19, B=@n20))) {name=GM-DesertPlants, nodeColor=<rgba 0.5 0 1 1>}
  n21 = Max(A=NamedRerouteUsage(), B=NamedRerouteUsage())
  n22 = LandscapeLayerSample() {parameterName=BiomeC}
  n23 = Constant() {r=3}
  n24 = Max(
   A=Max(A=LandscapeLayerSample() {parameterName=BiomeA}, B=LandscapeLayerSample() {parameterName=BiomeB})
   B=Saturate(None=Multiply(A=@n22, B=@n23))
  )
  NamedRerouteDeclaration(None=OneMinus(None=Max(A=@n21, B=@n24))) {name=GM-Grass, nodeColor=<rgba 0.5 0 1 1>}
  n25 = LandscapeLayerSample() {parameterName=Sand}
  n26 = Max(
   A=Max(A=LandscapeLayerSample() {parameterName=BiomeA}, B=LandscapeLayerSample() {desc=DesertScree, parameterName=BiomeC})
   B=LandscapeLayerSample() {desc=DesertRock, parameterName=BiomeB}
  )
  n27 = Constant() {r=20}
  n28 = Max(A=Max(A=NamedRerouteUsage(), B=NamedRerouteUsage()), B=Saturate(None=Multiply(A=@n26, B=@n27)))
  n29 = Subtract(A=ComponentMask(None=WorldPosition().XYZ) {b=true}, B=Constant() {r=200})
  n30 = Constant() {r=1}
  n31 = Subtract(A=Saturate(None=Subtract(A=@n25, B=@n28)), B=Saturate(None=Divide(A=@n29, B=@n30)))
  n32 = ComponentMask(None=WorldPosition().XYZ) {b=true}
  n33 = Constant() {r=-1200}
  n34 = OneMinus(None=Saturate(None=Subtract(A=@n32, B=@n33)))
  NamedRerouteDeclaration(None=Subtract(A=Saturate(None=@n31), B=@n34)) {name=GM-Beach, nodeColor=<rgba 0.5 0 1 1>}
  n35 = Saturate(None=Subtract(A=@n32, B=Constant() {r=-700}))
  NamedRerouteDeclaration(None=Subtract(A=OneMinus(None=@n35), B=@n34)) {name=GM-Ocean, nodeColor=<rgba 0.5 0 1 1>}
  n36 = Saturate(None=LinearInterpolate(A=Constant() {r=-2}, B=Constant() {r=4}, Alpha=LandscapeLayerSample() {desc=DesertScree, parameterName=BiomeC}))
  n37 = VertexNormalWS()
  n38 = Saturate(None=LinearInterpolate(A=Constant() {r=-4}, B=Constant() {r=1.6832}, Alpha=ComponentMask(None=@n37) {b=true}))
  n39 = Sign(None=LandscapeLayerSample() {parameterName=BiomeD})
  NamedRerouteDeclaration(None=Multiply(A=Multiply(A=@n36, B=@n38), B=OneMinus(None=@n39))) {name=GM-DesertScree, nodeColor=<rgba 0.5 0 1 1>}
  NamedRerouteDeclaration(None=@n18.GravelMaskForGrassFoliage) {name=GM-Gravel, nodeColor=<rgba 0.5 0 1 1>}
  n40 = MaterialFunctionCall!(
   Scale=ScalarParameter() {defaultValue=1.5, parameterName=DesertSand-Scale}
   Diffuse=TextureObject() {texture=/Game/Environments/Artemis/Sets/Terrain/Textures/T_Artemis_Desert_SandWavy_02_D}
   Normal=TextureObject() {samplerType=SAMPLERTYPE_Normal, texture=/Game/Environments/Artemis/Sets/Terrain/Textures/T_Artemis_Desert_SandWavy_02_N}
   S_Mask=TextureObject() {samplerType=SAMPLERTYPE_Masks, texture=/Game/Environments/Artemis/Sets/Terrain/Textures/T_Artemis_Desert_SandWavy_02_S}
  ) {materialFunction=/Game/Athena/Apollo/Environments/Landscape/Materials/Biome/UniqueBiomeMaterialFunctions/MF_Biome_Generic}
  n41 = MaterialFunctionCall!(
   Material 1=@n18.Material
   Height Texture 1=@n18.Material
   Material 2=@n40.Material
   Height Texture 2=@n40.Material
   Transistion Phase=LandscapeLayerSample() {parameterName=BiomeA}
   UseContrast
   Contrast
   AddLayerMask
  ) {materialFunction=/Game/Athena/Apollo/Environments/Landscape/Materials/Biome/CommonMaterials/MF_BlendLandscapeLayers}
  n42 = MaterialFunctionCall(
   Scale=ScalarParameter() {defaultValue=1.3, parameterName=DesertRock-Scale}
   Diffuse=TextureObject() {texture=/Game/Environments/Artemis/Sets/Terrain/Textures/T_Artemis_Desert_Ground_Rock_01_D}
   Normal=TextureObject() {samplerType=SAMPLERTYPE_Normal, texture=/Game/Environments/Artemis/Sets/Terrain/Textures/T_Artemis_Desert_Ground_Rock_01_N}
   S_Mask=TextureObject() {samplerType=SAMPLERTYPE_Masks, texture=/Game/Environments/Artemis/Sets/Terrain/Textures/T_Artemis_Desert_Ground_Rock_01_S}
  ) {materialFunction=&MF_Biome_Generic}
  n43 = MaterialFunctionCall(
   Material 1=@n41.BlendedMaterial
   Height Texture 1=@n41.BlendedMaterial
   Material 2=@n42.Material
   Height Texture 2=@n42.Material
   Transistion Phase=LandscapeLayerSample() {desc=DesertRock, parameterName=BiomeB}
   UseContrast=StaticBool() {value=true}
   Contrast=ScalarParameter() {defaultValue=1, parameterName=DesertRock-Contrast}
  ) {materialFunction=&MF_BlendLandscapeLayers}
  n44 = MaterialFunctionCall(
   Scale=ScalarParameter() {defaultValue=1.33, parameterName=DesertScree-Scale}
   Diffuse=TextureObject() {texture=/Game/Environments/Artemis/Sets/Terrain/Textures/T_Artemis_Desert_Gravel_01_D}
   Normal=TextureObject() {samplerType=SAMPLERTYPE_Normal, texture=/Game/Environments/Artemis/Sets/Terrain/Textures/T_Artemis_Desert_Gravel_01_N}
   S_Mask=TextureObject() {samplerType=SAMPLERTYPE_Masks, texture=/Game/Environments/Artemis/Sets/Terrain/Textures/T_Artemis_Desert_Gravel_01_S}
  ) {materialFunction=&MF_Biome_Generic}
  n45 = MaterialFunctionCall(
   Material 1=@n43.BlendedMaterial
   Height Texture 1=@n43.BlendedMaterial
   Material 2=@n44.Material
   Height Texture 2=@n44.Material
   Transistion Phase=LandscapeLayerSample() {desc=DesertScree, parameterName=BiomeC}
   UseContrast=StaticBool() {value=true}
   Contrast=ScalarParameter() {defaultValue=2, parameterName=DesertScree-Contrast}
  ) {materialFunction=&MF_BlendLandscapeLayers}
  n46 = ComponentMask(None=VertexNormalWS()) {b=true}
  n47 = Saturate(None=LinearInterpolate(
    A=ScalarParameter() {defaultValue=3.5, parameterName=Desert-test-cliff-slope-lerpA}
    B=ScalarParameter() {defaultValue=-0.3, parameterName=Desert-test-cliff-slope-lerpB}
    Alpha=Abs(None=@n46)
   ))
  n48 = MaterialFunctionCall!(Slope=@n47) {materialFunction=/Game/Athena/Apollo/Environments/Landscape/Materials/Biome/UniqueBiomeMaterialFunctions/MF_Desert_Cliff}
  n49 = MaterialFunctionCall!(
   Material=@n45.BlendedMaterial
   Heightmap=@n45.BlendedMaterial
   NewSlopeMaskMask=@n47
   Displacement
   AsphaltLayer=LandscapeLayerSample() {parameterName=Sand}
   UseAlternateCliffMat=StaticBool() {value=true}
   AltCliffHeight=@n48.Height
   AltCliffMat=@n48.Height
  ) {materialFunction=/Game/Athena/Apollo/Environments/Landscape/Materials/Biome/CommonMaterials/MF_SlopeBlendCliff}
  NamedRerouteDeclaration(None=@n49.HeightmapForGrass) {name=HeightmapForGrass, nodeColor=<rgba 1 1 0 1>}
  NamedRerouteDeclaration(None=@n49.CliffMask) {name=CliffMask, nodeColor=<rgba 0.5 0 1 1>}
  NamedRerouteDeclaration(None=@n49.LandscapeGrassMeshMask) {name=GM-Cliff, nodeColor=<rgba 0.5 0 1 1>}
  n50 = GetMaterialAttributes(None=MaterialFunctionCall!(PutSandBeforeGravelOverride, UseSnow=StaticBool(), UseOceanLevel) {materialFunction=/Game/Athena/Apollo/Environments/Landscape/Materials/Biome/CommonMaterials/Mobile/MF_Mobile_CommonLandscapeLayers}.Material) {attributeGetTypes [#1]}
  MaterialExpressionConstant_2 = Constant()
  n51 = LandscapeLayerBlend!(
   Layer Gravel=@MaterialExpressionConstant_2
   Layer Mud=@MaterialExpressionConstant_2
   Layer Sand=@MaterialExpressionConstant_2
   Layer Asphalt=@MaterialExpressionConstant_2
  ) {layers [
    {
     blendType=LB_WeightBlend
     constHeightInput=0
     layerName=Gravel
     previewWeight=0
     constLayerInput=<0 0 0>
     heightInput {expression=null, inputName=None, mask=0, maskA=0, maskB=0, maskG=0, maskR=0, outputIndex=0}
     layerInput {
      expression=&self:MaterialExpressionConstant_2
      inputName=None
      mask=0
      maskA=0
      maskB=0
      maskG=0
      maskR=0
      outputIndex=0
     }
    }
    {
     blendType=LB_WeightBlend
     constHeightInput=0
     layerName=Mud
     previewWeight=0
     constLayerInput=<0 0 0>
     heightInput {expression=null, inputName=None, mask=0, maskA=0, maskB=0, maskG=0, maskR=0, outputIndex=0}
     layerInput {
      expression=&self:MaterialExpressionConstant_2
      inputName=None
      mask=0
      maskA=0
      maskB=0
      maskG=0
      maskR=0
      outputIndex=0
     }
    }
    {
     blendType=LB_WeightBlend
     constHeightInput=0
     layerName=Sand
     previewWeight=0
     constLayerInput=<0 0 0>
     heightInput {expression=null, inputName=None, mask=0, maskA=0, maskB=0, maskG=0, maskR=0, outputIndex=0}
     layerInput {
      expression=&self:MaterialExpressionConstant_2
      inputName=None
      mask=0
      maskA=0
      maskB=0
      maskG=0
      maskR=0
      outputIndex=0
     }
    }
    {
     blendType=LB_WeightBlend
     constHeightInput=0
     layerName=Asphalt
     previewWeight=0
     constLayerInput=<0 0 0>
     heightInput {expression=null, inputName=None, mask=0, maskA=0, maskB=0, maskG=0, maskR=0, outputIndex=0}
     layerInput {
      expression=&self:MaterialExpressionConstant_2
      inputName=None
      mask=0
      maskA=0
      maskB=0
      maskG=0
      maskR=0
      outputIndex=0
     }
    }
   ]}
  n52 = SetMaterialAttributes!(MaterialAttributes=@n50.MaterialAttributes, Base Color=Add(A=@n50.Base Color, B=Multiply(A=@n51) {constB=0})) {desc="This must remain in material,\r\nnot a function.", attributeSetTypes [#1]}
  n53 = MaterialFunctionCall!(Material=MaterialFunctionCall!(MA=@n49.Material, Mask=LandscapeLayerSample() {parameterName=BiomeD}, HeightIn=@n49.Material, SlopeMask=@n49.Material, BlendLandscapeLayers) {materialFunction=/Game/TechArt/Seasonal/LackLuster/MaterialFunctions/MF_LackLuster_Landscape_RVT}.Result) {materialFunction=/Game/Athena/Apollo/Maps/Landscape/VT/MaterialFunctions/TransformTangentToWorld}
  n54 = GetMaterialAttributes(None=MaterialFunctionCall!(MobileMaterial=@n52, HighEndMaterial=NamedRerouteDeclaration(None=@n53.Result) {name=MaterialForRVT, nodeColor=<rgba 0.5 0 1 1>}) {materialFunction=/Game/Athena/Asteria/Environments/Landscape/Materials/MaterialFunctions/FN_RvtWriteSwitch}.Result) {attributeGetTypes [
    #1
    #2
    #3
    #4
   ]}
  RuntimeVirtualTextureOutput!(
   BaseColor=MaterialFunctionCall!(BaseColor=@n54.Base Color) {materialFunction=/Game/Athena/Apollo/Environments/Landscape/Materials/Biome/CommonMaterials/MF_ReduceRTV-CompressionArtifacts}.BaseColor
   Specular=@n54.Specular
   Roughness=@n54.Specular
   Normal=@n54.Specular
   WorldHeight=ComponentMask(None=WorldPosition().XYZ) {b=true}
   Opacity
   Mask
   Displacement
   Mask4
  )
  Saturate(None=LinearInterpolate(A=@n17, B=ScalarParameter() {parameterName=DesertScree-NoiseLerpB}, Alpha=@n16.R))
  n55 = StaticBool()
  n56 = RuntimeVirtualTextureSampleParameter() {
   group=RVT
   materialType=BaseColor_Normal_Specular_YCoCg
   parameterName=RVT Sample Target
   virtualTexture=/Game/Athena/Apollo/Maps/Landscape/VT/RVT_Apollo_Landscape
  }
  n57 = SetMaterialAttributes!(MaterialAttributes, Base Color=@n56.BaseColor, Specular=@n56.BaseColor, Roughness=@n56.BaseColor, Normal=@n56.BaseColor) {attributeSetTypes [
    #1
    #2
    #3
    #4
   ]}
  n58 = MaterialFunctionCall!(
   In=MaterialFunctionCall!(
    IsStaticSnow
    UseAlternateStaticSnowMask
    AlternateStaticSnowMask
    "UseRoad&&OcclusionMask"
    IsForGrassMeshToPushDown
    AdditionalMaskToRemoveSnow
    IsFallingSnowFX
    UseMinimapCoords
    IncreaseSnowUsingSlope
    UseSmoothBorder
    LerpHighM
    TemporalHeightLerpContrast
    IsLobbyIsland
    UseBR-IslandMask=@n55
    IsTree
    IsForRenderTarget
    UseSparkle
    UseSnowNormals=StaticBool() {value=true}
    UseAdditionalMaskToRemoveSnow
    IsTangentSpaceInput
    UseVertexNormalsForBasicSnow
    UseTangentSnowNormals=@n55
    Material=MaterialFunctionCall(NoRVT=@n52, RVT=@n57) {materialFunction=&FN_RVT_Switch_Material}.Result
   ) {materialFunction=/Game/Athena/Apollo/Environments/Landscape/Materials/Biome/UniqueBiomeMaterialFunctions/MF_Apollo_Snow_Progression}.Material
   Mask=LandscapeLayerSample() {parameterName=BiomeD}
   UseCheapSlopeMask
   MaskReflectionCurveOnZ
   HueShiftOverride
  ) {materialFunction=/Game/TechArt/Seasonal/LackLuster/MaterialFunctions/MF_LackLuster_Landscape_PostRVT}
  SetMaterialAttributes!(
   MaterialAttributes=MaterialFunctionCall!(
    Material=MaterialFunctionCall!(
     In=@n58.Result
     UseHeightFalloff
     OnFireEmissiveOverride
     HeightFalloffRadius
     HeightFalloffHardness
     VariationMasksMobile
     In_Grass
     Is_Stylised_WE
    ) {desc=Fire, materialFunction=/Game/Packages/Fortress_SharedMaterials/Base_Material_Functions/MF_Curie_Landscape}.Result
    DebugInputAsBasecolor
    Non-RVT material=NamedRerouteUsage()
    DebugMaterialAs1Color=VectorParameter() {parameterName=DebugBiomeColor, defaultValue=<rgba 0.8 0.3 0 1>}.RGB
   ) {materialFunction=/Game/Athena/Apollo/Environments/Landscape/Materials/Biome/CommonMaterials/MF_LS_DEBUG}.Result
   Opacity Mask=LandscapeVisibilityMask()
  ) {attributeSetTypes [#5]}
  TextureObject() {texture=/Game/Environments/Artemis/Sets/Terrain/Textures/T_Artemis_Desert_Ground_Rock_02_D}
  TextureObject() {samplerType=SAMPLERTYPE_Normal, texture=/Game/Environments/Artemis/Sets/Terrain/Textures/T_Artemis_Desert_Ground_Rock_02_N}
  TextureObject() {samplerType=SAMPLERTYPE_Masks, texture=/Game/Environments/Artemis/Sets/Terrain/Textures/T_Artemis_Desert_Ground_Rock_02_S}
 ]
}
