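# Landscape material function for Frontier moss in the Apollo biome, outputting material attributes plus a height from texture alpha.
# Named reroutes supply base color from mask-blended tints and curve-atlas detail, with a Fresnel on the macro normal darkening cores and brightening fuzzy edges.
# Bump-offset UVs sample masks for specular and roughness, plus a micro normal angle-corrected onto the macro normal.
# ColorMask4, world-position color variation and lump-mask branches are unconnected.
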MaterialFunction /Game/Athena/Apollo/Environments/Landscape/Materials/Biome/UniqueBiomeMaterialFunctions/MF_FrontierMoss
graph {nodes [
  Constant3Vector() {constant=<rgba 1.0666 0.3094 0.5924 1>}
  n1 = Constant() {r=1}
  FunctionInput(Preview=@n1) {inputName=ColorMask4, inputType=FunctionInput_Scalar}
  n2 = Constant() {r=0.01}
  n3 = NamedRerouteUsage()
  n4 = NamedRerouteUsage()
  n5 = Saturate(None=Add(A=Multiply(A=@n2, B=@n3), B=Multiply(A=@n3, B=@n4)))
  n6 = ScalarParameter() {defaultValue=0.25, group=FrontierMoss, parameterName=Moss_CoreDarkening}
  n7 = ScalarParameter() {defaultValue=1.5, group=FrontierMoss, parameterName=Moss_FuzzyEdgeBrightness}
  n8 = LandscapeLayerCoords() {mappingScale=10}
  n9 = ScalarParameter() {defaultValue=1, group=FrontierMoss, parameterName=FuzzNormalScale}
  n10 = MaterialFunctionCall(
   Normal=TextureSample(UVs=Multiply(A=@n8, B=@n9)) {samplerType=SAMPLERTYPE_Normal, texture=/Game/Packages/Fortress_Sky/Materials/2Normal}.RGB
   Flatness=ScalarParameter() {defaultValue=0.5, group=FrontierMoss, parameterName=MAcroNormalFlatter}
  ) {materialFunction=&FlattenNormal}
  n11 = Power(
   Base=Fresnel(ExponentIn=ScalarParameter() {defaultValue=1, group=FrontierMoss, parameterName=FuzzExponent}, Normal=Transform(None=@n10.Result))
   Exp=ScalarParameter() {defaultValue=2, group=FrontierMoss, parameterName=FuzzEdgeClamp}
  )
  n12 = Multiply(A=@n8, B=ScalarParameter() {defaultValue=2, group=FrontierMoss, parameterName=Moss_Scale})
  n13 = BumpOffset(
   Coordinate=@n12
   Height=TextureSample(UVs=@n12) {samplerType=SAMPLERTYPE_Masks, texture=/Game/Environments/Artemis/Sets/Terrain/Textures/T_Artemis_Frontier_Moss_01_D}.R
   HeightRatioInput=ScalarParameter() {defaultValue=-0.1, group=FrontierMoss, parameterName=TerrainHeightRatio}
  )
  n14 = TextureSample(UVs=@n13) {samplerType=SAMPLERTYPE_Masks, texture=/Game/Environments/Artemis/Sets/Terrain/Textures/T_Artemis_Frontier_Moss_01_S}
  n15 = TextureSample(UVs=@n13) {samplerType=SAMPLERTYPE_Normal, texture=/Game/Environments/Artemis/Sets/Terrain/Textures/T_Artemis_Frontier_Moss_01_N}
  n16 = ScalarParameter() {defaultValue=0.25, group=FrontierMoss, parameterName=MicroNormalFlatter}
  n17 = SetMaterialAttributes!(
   MaterialAttributes
   BaseColor=LinearInterpolate(A=Multiply(A=@n5, B=@n6), B=Multiply(A=@n5, B=@n7), Alpha=Clamp(None=@n11))
   Specular=LinearInterpolate(
    A=ScalarParameter() {defaultValue=0.5, group=FrontierMoss, parameterName=Moss_Min_Specular}
    B=ScalarParameter() {defaultValue=1, group=FrontierMoss, parameterName=Moss_Max_Specular}
    Alpha=@n14.R
   )
   Roughness=LinearInterpolate(
    A=ScalarParameter() {defaultValue=0.75, group=FrontierMoss, parameterName=Moss_Min_Roughness}
    B=ScalarParameter() {defaultValue=1, group=FrontierMoss, parameterName=Moss_Max_Roughness}
    Alpha=@n14.B
   )
   Normal=MaterialFunctionCall(
    BaseNormal=@n10.Result
    AdditionalNormal=MaterialFunctionCall(Normal=@n15.RGB, Flatness=@n16) {materialFunction=&FlattenNormal}.Result
   ) {materialFunction=&BlendAngleCorrectedNormals}.Result
  ) {attributeSetTypes [
    #1
    #2
    #3
    #4
   ]}
  FunctionOutput(None=@n17) {outputName=Material}
  n18 = TextureSample(UVs=@n12) {samplerType=SAMPLERTYPE_Masks, texture=&T_Artemis_Frontier_Moss_01_D}
  FunctionOutput(None=@n18.A) {outputName=Height}
  n19 = Constant3Vector() {constant=<rgba 0.6654 0.2918 0.5972 1>}
  n20 = WorldPosition() {worldPositionShaderOffset=WPT_ExcludeAllShaderOffsets}
  n21 = TextureSample(UVs=Divide(A=ComponentMask(None=@n20.XYZ) {g=true, r=true}, B=Constant() {r=-13000})) {
   samplerSource=SSM_Wrap_WorldGroupSettings
   samplerType=SAMPLERTYPE_Grayscale
   texture=/Game/Athena/Apollo/Environments/Landscape/Materials/Biome/CommonMaterials/T_ApolloGrass_ColorMask
  }
  LinearInterpolate(A=Constant3Vector() {constant=<rgba 0.581 0.2737 0.9062 1>}, B=@n19, Alpha=@n21.R)
  n22 = TextureSample(UVs=Multiply(A=@n8, B=ScalarParameter() {defaultValue=1, parameterName=MossLumpTiling})) {
   samplerType=SAMPLERTYPE_Grayscale
   texture=&T_ApolloGrass_ColorMask
  }
  Max(A=@n11, B=@n22.G)
  NamedRerouteDeclaration(None=@n19) {name=MacroColorLerp, nodeColor=<rgba 0 0.2266 1 1>}
  n23 = LinearInterpolate(
   A=VectorParameter() {group=FrontierMoss, parameterName=Moss_Color_01, defaultValue=<rgba 0.8398 0.3833 1.0666 1>}.RGB
   B=VectorParameter() {group=FrontierMoss, parameterName=Moss_Color_02, defaultValue=<rgba 0.6574 0.3958 0.6666 1>}.RGB
   Alpha=FunctionInput(Preview=@n1) {inputName=ColorMask1, inputType=FunctionInput_Scalar}
  )
  n24 = VectorParameter() {group=FrontierMoss, parameterName=Moss_Color_03, defaultValue=<rgba 0.8398 0.3833 1.0666 1>}
  n25 = FunctionInput(Preview=@n1) {inputName=ColorMask2, inputType=FunctionInput_Scalar}
  NamedRerouteDeclaration(None=LinearInterpolate(
    A=LinearInterpolate(A=@n23, B=@n24.RGB, Alpha=@n25)
    B=VectorParameter() {group=FrontierMoss, parameterName=Moss_Color_04, defaultValue=<rgba 0.6574 0.3958 0.6666 1>}.RGB
    Alpha=FunctionInput(Preview=@n1) {inputName=ColorMask3, inputType=FunctionInput_Scalar}
   )) {name=MossMicroColors, nodeColor=<rgba 1 0.5859 0 1>}
  n26 = CurveAtlasRowParameter(CurveTime=LinearInterpolate(A=ScalarParameter() {defaultValue=0.25, group=FrontierMoss, parameterName=DiffuseMin}, Alpha=@n18.R)) {
   atlas=/Game/Athena/Apollo/Environments/Landscape/Materials/Frontier/FrontierColorAtlas
   curve=/Game/Athena/Apollo/Environments/Landscape/Materials/Frontier/FrontierMossDetailColor
   group=FrontierMoss
   parameterName=FrontierMossDetailColor
  }
  NamedRerouteDeclaration(None=@n26) {name=MossBaseColorDetail, nodeColor=<rgba 1 0 0.6562 1>}
 ]}
