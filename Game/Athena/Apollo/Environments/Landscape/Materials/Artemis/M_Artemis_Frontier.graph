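# Landscape material for the Artemis Frontier biome, driving terrain shading, grass spawning and physical surfaces from painted layers.
# Common layers height-blend with forest floor, scree and moss, then slope cliffs and optional LackLuster chrome, before writing to a runtime virtual texture.
# A simplified mobile path tints moss instead.
# The surface samples that RVT or uses the mobile material, then adds snow, LackLuster, fire, debug and visibility masking.
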Material /Game/Athena/Apollo/Environments/Landscape/Materials/Artemis/M_Artemis_Frontier
graph {
 properties {
  bTangentSpaceNormal=false
  bUseMaterialAttributes=true
  bUsedWithVirtualHeightfieldMesh=true
  pixelDepthOffsetMode=PDOM_Legacy
  usedShadingModels=Default Lit
 }
 nodes [
  n1 = MaterialFunctionCall!(
   AdditionalGrassBorderColorMask=LandscapeLayerSample() {desc=ForestFloor, parameterName=BiomeA}
   PutSandBeforeGravel=StaticBool() {value=true}
   UseOceanLevel
  ) {materialFunction=/Game/Athena/Apollo/Environments/Landscape/Materials/Biome/CommonMaterials/MF_CommonLandscapeLayers}
  n2 = MaterialFunctionCall() {materialFunction=/Game/Athena/Apollo/Environments/Landscape/Materials/Biome/UniqueBiomeMaterialFunctions/MF_Biome_Forest}
  n3 = MaterialFunctionCall!(
   Material 1=@n1.Material
   Height Texture 1=@n1.Material
   Material 2=@n2.Material
   Height Texture 2=NamedRerouteDeclaration(None=@n2.Height) {name=BiomeA-ForestFloorHeight, nodeColor=<rgba 0.5 0 1 1>}
   Transistion Phase=LandscapeLayerSample() {desc=ForestFloor, parameterName=BiomeA}
   UseContrast
   Contrast
   AddLayerMask
  ) {materialFunction=/Game/Athena/Apollo/Environments/Landscape/Materials/Biome/CommonMaterials/MF_BlendLandscapeLayers}
  n4 = LandscapeLayerSample() {desc=Flowers, parameterName=BiomeB}
  n5 = MaterialFunctionCall!(LayerWeight=@n4) {materialFunction=/Game/Athena/Apollo/Environments/Landscape/Materials/Biome/UniqueBiomeMaterialFunctions/MF_Scree}
  n6 = MaterialFunctionCall(
   Material 1=@n3.BlendedMaterial
   Height Texture 1=@n3.BlendedMaterial
   Material 2=@n5.Material
   Height Texture 2=NamedRerouteDeclaration(None=@n5.Height) {name=MF_Scree-Height, nodeColor=<rgba 0.5 0 1 1>}
   Transistion Phase=@n4
   UseContrast=StaticBool() {value=true}
   Contrast=Constant() {r=0.5}
  ) {materialFunction=&MF_BlendLandscapeLayers}
  n7 = MaterialFunctionCall!(ColorMask2=NamedRerouteUsage(), ColorMask3=NamedRerouteUsage(), ColorMask4=NamedRerouteUsage(), ColorMask1=NamedRerouteUsage()) {materialFunction=/Game/Athena/Apollo/Environments/Landscape/Materials/Biome/UniqueBiomeMaterialFunctions/MF_FrontierMoss}
  n8 = MaterialFunctionCall(
   Material 1=@n6.BlendedMaterial
   Height Texture 1=@n6.BlendedMaterial
   Material 2=@n7.Material
   Height Texture 2=@n7.Material
   Transistion Phase=LandscapeLayerSample() {parameterName=BiomeC}
   AddLayerMask=Constant()
  ) {materialFunction=&MF_BlendLandscapeLayers}
  n9 = MaterialFunctionCall!(
   Material=@n8.BlendedMaterial
   Heightmap=@n8.BlendedMaterial
   NewSlopeMaskMask
   Displacement
   AsphaltLayer=LandscapeLayerSample() {parameterName=Sand}
   UseAlternateCliffMat
   AltCliffHeight
   AltCliffMat
  ) {materialFunction=/Game/Athena/Apollo/Environments/Landscape/Materials/Biome/CommonMaterials/MF_SlopeBlendCliff}
  n10 = LandscapeLayerSample() {parameterName=BiomeD}
  n11 = MaterialFunctionCall!(Material=StaticSwitchParameter(
    True=MaterialFunctionCall!(MA=@n9.Material, Mask=@n10, HeightIn=@n9.Material, SlopeMask=@n9.Material, BlendLandscapeLayers) {materialFunction=/Game/TechArt/Seasonal/LackLuster/MaterialFunctions/MF_LackLuster_Landscape_RVT}.Result
    False=@n9.Material
   ) {parameterName=UseLackluster}) {materialFunction=/Game/Athena/Apollo/Maps/Landscape/VT/MaterialFunctions/TransformTangentToWorld}
  n12 = GetMaterialAttributes(None=MaterialFunctionCall!(PutSandBeforeGravelOverride, UseSnow=StaticBool(), UseOceanLevel) {materialFunction=/Game/Athena/Apollo/Environments/Landscape/Materials/Biome/CommonMaterials/Mobile/MF_Mobile_CommonLandscapeLayers}.Material) {attributeGetTypes [#1]}
  n13 = Multiply(
   A=Desaturation(None=@n12.Base Color, Fraction=Constant() {r=1})
   B=VectorParameter() {parameterName=MobileMossTintColor, defaultValue=<rgba 0.5906 0.5906 0.8148 1>}.RGB
  )
  n14 = LandscapeLayerSample() {parameterName=BiomeC}
  MaterialExpressionConstant_13 = Constant()
  n15 = LandscapeLayerBlend!(
   Layer Gravel=@MaterialExpressionConstant_13
   Layer Mud=@MaterialExpressionConstant_13
   Layer Sand=@MaterialExpressionConstant_13
   Layer Asphalt=@MaterialExpressionConstant_13
  ) {layers [
    {
     blendType=LB_WeightBlend
     constHeightInput=0
     layerName=Gravel
     previewWeight=0
     constLayerInput=<0 0 0>
     heightInput {expression=null, inputName=None, mask=0, maskA=0, maskB=0, maskG=0, maskR=0, outputIndex=0}
     layerInput {
      expression=&self:MaterialExpressionConstant_13
      inputName=None
      mask=0
      maskA=0
      maskB=0
      maskG=0
      maskR=0
      outputIndex=0
     }
    }
    {
     blendType=LB_WeightBlend
     constHeightInput=0
     layerName=Mud
     previewWeight=0
     constLayerInput=<0 0 0>
     heightInput {expression=null, inputName=None, mask=0, maskA=0, maskB=0, maskG=0, maskR=0, outputIndex=0}
     layerInput {
      expression=&self:MaterialExpressionConstant_13
      inputName=None
      mask=0
      maskA=0
      maskB=0
      maskG=0
      maskR=0
      outputIndex=0
     }
    }
    {
     blendType=LB_WeightBlend
     constHeightInput=0
     layerName=Sand
     previewWeight=0
     constLayerInput=<0 0 0>
     heightInput {expression=null, inputName=None, mask=0, maskA=0, maskB=0, maskG=0, maskR=0, outputIndex=0}
     layerInput {
      expression=&self:MaterialExpressionConstant_13
      inputName=None
      mask=0
      maskA=0
      maskB=0
      maskG=0
      maskR=0
      outputIndex=0
     }
    }
    {
     blendType=LB_WeightBlend
     constHeightInput=0
     layerName=Asphalt
     previewWeight=0
     constLayerInput=<0 0 0>
     heightInput {expression=null, inputName=None, mask=0, maskA=0, maskB=0, maskG=0, maskR=0, outputIndex=0}
     layerInput {
      expression=&self:MaterialExpressionConstant_13
      inputName=None
      mask=0
      maskA=0
      maskB=0
      maskG=0
      maskR=0
      outputIndex=0
     }
    }
   ]}
  n16 = SetMaterialAttributes!(
   MaterialAttributes=@n12.MaterialAttributes
   Base Color=Add(A=LinearInterpolate(A=@n12.Base Color, B=@n13, Alpha=@n14), B=Multiply(A=@n15) {constB=0})
  ) {desc="This must remain in material,\r\nnot a function.", attributeSetTypes [#1]}
  n17 = GetMaterialAttributes(None=MaterialFunctionCall!(MobileMaterial=@n16, HighEndMaterial=NamedRerouteDeclaration(None=@n11.Result) {name=MaterialForRVT, nodeColor=<rgba 0.5 0 1 1>}) {materialFunction=/Game/Athena/Asteria/Environments/Landscape/Materials/MaterialFunctions/FN_RvtWriteSwitch}.Result) {attributeGetTypes [
    #1
    #2
    #3
    #4
   ]}
  n18 = WorldPosition() {worldPositionShaderOffset=WPT_ExcludeAllShaderOffsets}
  n19 = TextureSample(UVs=Divide(A=ComponentMask(None=@n18.XYZ) {g=true, r=true}, B=ScalarParameter() {defaultValue=80, parameterName=NR-testcompress-tile}) {constB=8192}) {samplerSource=SSM_Wrap_WorldGroupSettings, samplerType=SAMPLERTYPE_LinearColor, texture=/Game/Global/Textures/Default/Noise/T_Apollo_TCNoise}
  Add(A=@n17.Normal, B=Multiply(A=@n19.RGB, B=VectorParameter() {parameterName=NR-testcompress-rgb, defaultValue=<rgba 0.2 0.2 0 1>}.RGB))
  Constant3Vector() {constant=<rgba 0.03 0.03 0 1>}
  Constant() {r=50}
  n20 = Reroute()
  n21 = Reroute()
  n22 = Constant3Vector() {constant=<rgba 0 1 0 0>}
  n23 = If(A=Constant(), B=@n21, "A > B"=@n22, "A == B"=@n22, "A < B"=Constant3Vector() {constant=<rgba 1 0 0 0>})
  n24 = If(A=@n21, B=@n20, "A > B"=@n23, "A == B"=@n23, "A < B"=Constant3Vector() {constant=<rgba 0 0 1 0>})
  If(A=@n20, B=Reroute(), "A > B"=@n24, "A == B"=@n24, "A < B"=@n22)
  n25 = NamedRerouteUsage()
  n26 = NamedRerouteUsage()
  n27 = Max(A=@n26, B=NamedRerouteUsage())
  n28 = NamedRerouteUsage()
  n29 = LandscapeLayerSample() {parameterName=BiomeD}
  MaterialExpressionMaterialFunctionCall_9 = MaterialFunctionCall!(In=Subtract(A=Subtract(A=@n25, B=@n27), B=Max(A=@n28, B=@n29))) {materialFunction=/Game/Athena/Apollo/Environments/Landscape/Materials/Biome/CommonMaterials/MF_LS_GrassSwitch}
  MaterialExpressionMaterialFunctionCall_14 = MaterialFunctionCall(In=NamedRerouteUsage()) {materialFunction=&MF_LS_GrassSwitch}
  MaterialExpressionSubtract_1 = Subtract(
   A=MaterialFunctionCall(In=NamedRerouteUsage()) {materialFunction=&MF_LS_GrassSwitch}.Result
   B=LandscapeLayerSample() {parameterName=BiomeD}
  )
  MaterialExpressionMaterialFunctionCall_18 = MaterialFunctionCall(In=@n26) {materialFunction=&MF_LS_GrassSwitch}
  MaterialExpressionMaterialFunctionCall_19 = MaterialFunctionCall() {materialFunction=&MF_LS_GrassSwitch}
  MaterialExpressionConstant_20 = Constant()
  MaterialExpressionMaterialFunctionCall_21 = MaterialFunctionCall() {materialFunction=&MF_LS_GrassSwitch}
  MaterialExpressionMaterialFunctionCall_22 = MaterialFunctionCall(In=NamedRerouteUsage()) {materialFunction=&MF_LS_GrassSwitch}
  MaterialExpressionMaterialFunctionCall_23 = MaterialFunctionCall(In=NamedRerouteUsage()) {materialFunction=&MF_LS_GrassSwitch}
  MaterialExpressionMaterialFunctionCall_24 = MaterialFunctionCall() {materialFunction=&MF_LS_GrassSwitch}
  MaterialExpressionMaterialFunctionCall_25 = MaterialFunctionCall() {materialFunction=&MF_LS_GrassSwitch}
  MaterialExpressionMaterialFunctionCall_26 = MaterialFunctionCall() {materialFunction=&MF_LS_GrassSwitch}
  n30 = Max(A=NamedRerouteUsage(), B=NamedRerouteUsage())
  n31 = NamedRerouteUsage()
  MaterialExpressionMaterialFunctionCall_20 = MaterialFunctionCall(In=Subtract(A=Subtract(A=@n30, B=@n29), B=OneMinus(None=@n31))) {materialFunction=&MF_LS_GrassSwitch}
  LandscapeGrassOutput!(
   Grass_Standard=@MaterialExpressionMaterialFunctionCall_9.Result
   Bush=@MaterialExpressionMaterialFunctionCall_14.Result
   GrasslandFlowers=@MaterialExpressionSubtract_1
   Scree=@MaterialExpressionMaterialFunctionCall_18.Result
   Scree_Grass=@MaterialExpressionMaterialFunctionCall_19.Result
   Heather=@MaterialExpressionConstant_20
   ForestFloor=@MaterialExpressionMaterialFunctionCall_21.Result
   Beach=@MaterialExpressionMaterialFunctionCall_22.Result
   Ocean=@MaterialExpressionMaterialFunctionCall_23.Result
   DesertPlants=@MaterialExpressionMaterialFunctionCall_24.Result
   DesertScree=@MaterialExpressionMaterialFunctionCall_25.Result
   TropicalJungle=@MaterialExpressionMaterialFunctionCall_26.Result
   FrontierPlants=@MaterialExpressionMaterialFunctionCall_20.Result
  ) {
   desc="This must have the exact inputs in every biome material!\r\nThis data is baked into the level data when the level is\r\nsaved, so any changes to this require the affected level\r\nto be saved && submitted."
   grassTypes [
    Grass_Standard {
     grassType=/Game/Athena/Apollo/Environments/Landscape/SmallFoliage/DataAssets/LandscapeGrassType/LandscapeGrassType_StandardGrass
     input {
      expression=&self:MaterialExpressionMaterialFunctionCall_9
      inputName=None
      mask=0
      maskA=0
      maskB=0
      maskG=0
      maskR=0
      outputIndex=0
     }
    }
    Bush {
     grassType=/Game/Athena/Environments/Landscape/Foliage/Athena_Foliage_Bushes
     input {
      expression=&self:MaterialExpressionMaterialFunctionCall_14
      inputName=None
      mask=0
      maskA=0
      maskB=0
      maskG=0
      maskR=0
      outputIndex=0
     }
    }
    GrasslandFlowers {
     grassType=/Game/Athena/Apollo/Environments/Landscape/SmallFoliage/DataAssets/LandscapeGrassType/LandscapeGrassType_FarmFlowers
     input {
      expression=&self:MaterialExpressionSubtract_1
      inputName=None
      mask=0
      maskA=0
      maskB=0
      maskG=0
      maskR=0
      outputIndex=0
     }
    }
    Scree {
     grassType=/Game/Athena/Apollo/Environments/Landscape/SmallFoliage/DataAssets/LandscapeGrassType/LandscapeGrassType_Scree
     input {
      expression=&self:MaterialExpressionMaterialFunctionCall_18
      inputName=None
      mask=0
      maskA=0
      maskB=0
      maskG=0
      maskR=0
      outputIndex=0
     }
    }
    Scree_Grass {
     grassType=/Game/Athena/Apollo/Environments/Landscape/SmallFoliage/DataAssets/LandscapeGrassType/LandscapeGrassType_ScreeGrass
     input {
      expression=&self:MaterialExpressionMaterialFunctionCall_19
      inputName=None
      mask=0
      maskA=0
      maskB=0
      maskG=0
      maskR=0
      outputIndex=0
     }
    }
    Heather {
     grassType=/Game/Athena/Apollo/Environments/Landscape/SmallFoliage/DataAssets/LandscapeGrassType/LandscapeGrassType_Heather
     input {
      expression=&self:MaterialExpressionConstant_20
      inputName=None
      mask=0
      maskA=0
      maskB=0
      maskG=0
      maskR=0
      outputIndex=0
     }
    }
    ForestFloor {
     grassType=/Game/Athena/Apollo/Environments/Landscape/SmallFoliage/DataAssets/LandscapeGrassType/LandscapeGrassType_ForestGrass
     input {
      expression=&self:MaterialExpressionMaterialFunctionCall_21
      inputName=None
      mask=0
      maskA=0
      maskB=0
      maskG=0
      maskR=0
      outputIndex=0
     }
    }
    Beach {
     grassType=/Game/Athena/Apollo/Environments/Landscape/SmallFoliage/DataAssets/LandscapeGrassType/LandscapeGrassType_Beach
     input {
      expression=&self:MaterialExpressionMaterialFunctionCall_22
      inputName=None
      mask=0
      maskA=0
      maskB=0
      maskG=0
      maskR=0
      outputIndex=0
     }
    }
    Ocean {
     grassType=/Game/Athena/Apollo/Environments/Landscape/SmallFoliage/DataAssets/LandscapeGrassType/LandscapeGrassType_Ocean
     input {
      expression=&self:MaterialExpressionMaterialFunctionCall_23
      inputName=None
      mask=0
      maskA=0
      maskB=0
      maskG=0
      maskR=0
      outputIndex=0
     }
    }
    DesertPlants {
     grassType=/Game/Athena/Apollo/Environments/Landscape/SmallFoliage/DataAssets/LandscapeGrassType/LandscapeGrassType_DesertPlants
     input {
      expression=&self:MaterialExpressionMaterialFunctionCall_24
      inputName=None
      mask=0
      maskA=0
      maskB=0
      maskG=0
      maskR=0
      outputIndex=0
     }
    }
    DesertScree {
     grassType=/Game/Athena/Apollo/Environments/Landscape/SmallFoliage/DataAssets/LandscapeGrassType/LandscapeGrassType_DesertScree
     input {
      expression=&self:MaterialExpressionMaterialFunctionCall_25
      inputName=None
      mask=0
      maskA=0
      maskB=0
      maskG=0
      maskR=0
      outputIndex=0
     }
    }
    TropicalJungle {
     grassType=/Game/Athena/Apollo/Environments/Landscape/SmallFoliage/DataAssets/LandscapeGrassType/LandscapeGrassType_Nightmare
     input {
      expression=&self:MaterialExpressionMaterialFunctionCall_26
      inputName=None
      mask=0
      maskA=0
      maskB=0
      maskG=0
      maskR=0
      outputIndex=0
     }
    }
    FrontierPlants {
     grassType=/Game/Athena/Apollo/Environments/Landscape/SmallFoliage/DataAssets/LandscapeGrassType/NewLandscapeGrassType_FrontierGrass
     input {
      expression=&self:MaterialExpressionMaterialFunctionCall_20
      inputName=None
      mask=0
      maskA=0
      maskB=0
      maskG=0
      maskR=0
      outputIndex=0
     }
    }
   ]
  }
  MaterialExpressionMaterialFunctionCall_7 = MaterialFunctionCall!(SampleValue=LandscapeLayerSample() {parameterName=BiomeD}) {materialFunction=/Game/Athena/Apollo/Environments/Landscape/Materials/MaterialFunctions/MF_ArtemisChromePhysicalMaterialMask}
  n32 = NamedRerouteUsage()
  n33 = OneMinus(None=@n32) {desc="White=Cliffs"}
  n34 = LandscapeLayerSample() {parameterName=Gravel}
  n35 = LandscapeLayerSample() {parameterName=Mud}
  n36 = Saturate(None=Multiply(A=Max(A=@n34, B=@n35), B=Constant() {r=2}))
  n37 = Subtract(A=@n36, B=@n33) {desc="Remove Cliff from Gravel,Mud,Sand"}
  n38 = LandscapeLayerSample() {parameterName=Sand}
  n39 = Saturate(None=Multiply(A=LandscapeLayerSample() {parameterName=Asphalt}, B=Constant() {r=3}))
  n40 = Max(A=Max(A=NamedRerouteUsage(), B=NamedRerouteUsage()), B=@n39)
  MaterialExpressionSaturate_12 = Saturate(None=Subtract(A=Saturate(None=@n37), B=Max(A=@n38, B=@n40)) {desc="Remove ForestFloor, Flowers, Asphalt, Sand"}) {desc=Final Dirt mask}
  MaterialExpressionReroute_25 = Reroute(None=@n33) {desc=Final Stone mask}
  n41 = Subtract(A=@n32, B=Max(A=@n38, B=@n36)) {desc="Remove Gravel,Mud,Sand"}
  MaterialExpressionSaturate_14 = Saturate(None=Subtract(A=Saturate(None=@n41), B=@n39) {desc=Remove Asphalt}) {desc=Final Grass mask}
  MaterialExpressionReroute_29 = Reroute(None=@n39)
  n42 = Saturate(None=Subtract(A=@n38, B=@n33) {desc=Remove Cliff})
  MaterialExpressionReroute_42 = Reroute(None=Saturate(None=Subtract(A=@n42, B=@n40) {desc="Remove ForestFloor, Flowers, Asphalt"}))
  LandscapePhysicalMaterialOutput!(
   Chrome_PhysicalMaterial=@MaterialExpressionMaterialFunctionCall_7.Result
   Dirt=@MaterialExpressionSaturate_12
   Stone=@MaterialExpressionReroute_25
   Grass=@MaterialExpressionSaturate_14
   Asphalt=@MaterialExpressionReroute_29
   Sand=@MaterialExpressionReroute_42
  ) {
   desc="These are layered in order of importance.\r\nIf Dirt=1 and Stone=1, then Dirt will \"win\".\r\nTheoretically, these values aren't clamped.\r\nSo a value of 2 should trump a value of 1."
   inputs [
    {
     physicalMaterial=/Game/TechArt/Seasonal/LackLuster/Physics/Chrome_PhysicalMaterial
     input {
      expression=&self:MaterialExpressionMaterialFunctionCall_7
      inputName=None
      mask=0
      maskA=0
      maskB=0
      maskG=0
      maskR=0
      outputIndex=0
     }
    }
    {
     physicalMaterial=/Game/FortressPhysicalMaterials/Dirt
     input {
      expression=&self:MaterialExpressionSaturate_12
      inputName=None
      mask=0
      maskA=0
      maskB=0
      maskG=0
      maskR=0
      outputIndex=0
     }
    }
    {
     physicalMaterial=/Game/FortressPhysicalMaterials/Stone
     input {
      expression=&self:MaterialExpressionReroute_25
      inputName=None
      mask=0
      maskA=0
      maskB=0
      maskG=0
      maskR=0
      outputIndex=0
     }
    }
    {
     physicalMaterial=/Game/FortressPhysicalMaterials/Grass
     input {
      expression=&self:MaterialExpressionSaturate_14
      inputName=None
      mask=0
      maskA=0
      maskB=0
      maskG=0
      maskR=0
      outputIndex=0
     }
    }
    {
     physicalMaterial=/Game/FortressPhysicalMaterials/Asphalt
     input {
      expression=&self:MaterialExpressionReroute_29
      inputName=None
      mask=0
      maskA=0
      maskB=0
      maskG=0
      maskR=0
      outputIndex=0
     }
    }
    {
     physicalMaterial=/Game/FortressPhysicalMaterials/Sand
     input {
      expression=&self:MaterialExpressionReroute_42
      inputName=None
      mask=0
      maskA=0
      maskB=0
      maskG=0
      maskR=0
      outputIndex=0
     }
    }
   ]
  }
  n43 = LinearInterpolate(A=@MaterialExpressionSaturate_12, B=@MaterialExpressionReroute_25, Alpha=ScalarParameter() {parameterName=DEBUG-PhysMat1})
  n44 = ScalarParameter() {parameterName=DEBUG-PhysMat2}
  LinearInterpolate(
   A=LinearInterpolate(
    A=LinearInterpolate(A=@n43, B=@MaterialExpressionSaturate_14, Alpha=@n44)
    B=@MaterialExpressionReroute_29
    Alpha=ScalarParameter() {parameterName=DEBUG-PhysMat3}
   )
   B=@MaterialExpressionReroute_42
   Alpha=ScalarParameter() {defaultValue=1, parameterName=DEBUG-PhysMat4}
  )
  NamedRerouteDeclaration(None=@n1.GrassNoiseMask) {name=GrassNoiseMask, nodeColor=<rgba 0.5 0 1 1>}
  NamedRerouteDeclaration(None=@n1.GrassBaseColorDetail) {name=GrassBaseColorDetail, nodeColor=<rgba 0.5 0 1 1>}
  NamedRerouteDeclaration(None=@n1.Material) {name=CommonMaterial, nodeColor=<rgba 0.5 0 1 1>}
  NamedRerouteDeclaration(None=@n9.SlopeScreeMask) {name=SlopeScreeMask, nodeColor=<rgba 0.5 0 1 1>}
  n45 = NamedRerouteUsage()
  n46 = NamedRerouteUsage()
  n47 = OneMinus(None=Max(A=Max(A=@n45, B=@n46), B=NamedRerouteUsage()))
  NamedRerouteDeclaration(None=@n47) {name=GM-Grass, nodeColor=<rgba 0.5 0 1 1>}
  NamedRerouteDeclaration(None=@n47) {name=GM-Bush, nodeColor=<rgba 0.5 0 1 1>}
  n48 = LandscapeLayerSample() {parameterName=Sand}
  n49 = Max(A=NamedRerouteUsage(), B=NamedRerouteUsage())
  n50 = ComponentMask(None=WorldPosition().XYZ) {b=true}
  n51 = Constant() {r=200}
  n52 = Subtract(A=Saturate(None=Subtract(A=@n48, B=@n49)), B=Saturate(None=Divide(A=@n50, B=@n51)))
  n53 = ComponentMask(None=WorldPosition().XYZ) {b=true}
  n54 = Constant() {r=-1200}
  n55 = OneMinus(None=Saturate(None=Subtract(A=@n53, B=@n54)))
  NamedRerouteDeclaration(None=Subtract(A=Saturate(None=@n52), B=@n55)) {name=GM-Beach, nodeColor=<rgba 0.5 0 1 1>}
  n56 = Saturate(None=Subtract(A=@n53, B=Constant() {r=-700}))
  NamedRerouteDeclaration(None=Subtract(A=OneMinus(None=@n56), B=@n55)) {name=GM-Ocean, nodeColor=<rgba 0.5 0 1 1>}
  NamedRerouteDeclaration(None=@n1.CommonLayerMaskForGrassFoliage) {name=GM-CommonLayers, nodeColor=<rgba 0.5 0 1 1>}
  NamedRerouteDeclaration(None=@n1.GrassHeightmap) {name=GrassHeightmap, nodeColor=<rgba 0.5 0 1 1>}
  NamedRerouteDeclaration(None=@n8.LayerMask) {name=GM-FrontierMoss, nodeColor=<rgba 1 0 0.7266 1>}
  NamedRerouteDeclaration(None=@n1.GrassColorMask1) {name=GrassColorMask1, nodeColor=<rgba 0.5 0 1 1>}
  NamedRerouteDeclaration(None=@n1.GrassColorMask2) {name=GrassColorMask2, nodeColor=<rgba 0.5 0 1 1>}
  NamedRerouteDeclaration(None=@n1.GrassColorMask3) {name=GrassColorMask3, nodeColor=<rgba 0.5 0 1 1>}
  n57 = Saturate(None=Multiply(A=Constant() {r=3}, B=LandscapeLayerSample() {desc=Scree, parameterName=BiomeB}))
  n58 = NamedRerouteUsage()
  NamedRerouteDeclaration(None=Subtract(A=Subtract(A=@n57, B=@n58), B=NamedRerouteUsage())) {name=GM-ScreeGrass, nodeColor=<rgba 0.5 0 1 1>}
  n59 = NamedRerouteUsage()
  n60 = Constant() {r=-4}
  n61 = Constant() {r=1.6832}
  n62 = ComponentMask(None=VertexNormalWS()) {b=true}
  n63 = Multiply(A=LandscapeLayerSample() {desc=Scree, parameterName=BiomeB}, B=Constant() {r=2})
  n64 = Constant() {r=-0.5}
  n65 = Multiply(A=Saturate(None=LinearInterpolate(A=@n60, B=@n61, Alpha=@n62)), B=Saturate(None=Add(A=@n63, B=@n64)))
  n66 = Sign(None=LandscapeLayerSample() {parameterName=BiomeD})
  NamedRerouteDeclaration(None=Multiply(A=Max(A=@n59, B=@n65), B=OneMinus(None=@n66))) {name=GM-Scree, nodeColor=<rgba 0.5 0 1 1>}
  NamedRerouteDeclaration(None=@n1.GravelMaskForGrassFoliage) {name=GM-Gravel, nodeColor=<rgba 0.5 0 1 1>}
  NamedRerouteDeclaration(None=@n9.SlopeMask) {name=SlopeMask, nodeColor=<rgba 1 1 0 1>}
  n67 = LinearInterpolate(
   A=ScalarParameter() {defaultValue=-0.5, parameterName=BiomeA-FFSpawnLa}
   B=ScalarParameter() {defaultValue=1.5, parameterName=BiomeA-FFSpawnLb}
   Alpha=LandscapeLayerSample() {desc=ForestFloor, parameterName=BiomeA}
  )
  NamedRerouteDeclaration(None=MaterialFunctionCall(
    Transition Phase=Saturate(None=@n67)
    Height Texture=NamedRerouteUsage()
    Contrast=ScalarParameter() {defaultValue=0.5, parameterName=BiomeA-FFSpawnCont}
   ) {desc=ForestFloor, materialFunction=&HeightLerp}.Alpha) {name=GM-ForestFloor, nodeColor=<rgba 0.5 0 1 1>}
  NamedRerouteDeclaration(None=@n9.CliffMask) {name=CliffMask, nodeColor=<rgba 0.5 0 1 1>}
  NamedRerouteDeclaration(None=@n9.LandscapeGrassMeshMask) {name=GM-Cliff, nodeColor=<rgba 0.5 0 1 1>}
  NamedRerouteDeclaration(None=Constant()) {name=GM-Flowers, nodeColor=<rgba 0.5 0 1 1>}
  NamedRerouteUsage()
  RuntimeVirtualTextureOutput!(
   BaseColor=MaterialFunctionCall!(BaseColor=@n17.Base Color) {materialFunction=/Game/Athena/Apollo/Environments/Landscape/Materials/Biome/CommonMaterials/MF_ReduceRTV-CompressionArtifacts}.BaseColor
   Specular=@n17.Specular
   Roughness=@n17.Specular
   Normal=@n17.Specular
   WorldHeight=ComponentMask(None=WorldPosition().XYZ) {b=true}
   Opacity
   Mask
   Displacement
   Mask4
  )
  n68 = LandscapeLayerSample() {desc=Flowers, parameterName=BiomeB}
  n69 = Constant() {r=2}
  Saturate(None=Add(A=Multiply(A=@n68, B=@n69), B=Constant() {r=-0.5}))
  ScalarParameter() {defaultValue=1.6832, parameterName=Threshold_Scree2}
  ScalarParameter() {defaultValue=-4, parameterName=Threshold_Scree}
  n70 = StaticBool()
  n71 = RuntimeVirtualTextureSampleParameter() {
   group=RVT
   materialType=BaseColor_Normal_Specular_YCoCg
   parameterName=RVT Sample Target
   virtualTexture=/Game/Athena/Apollo/Maps/Landscape/VT/RVT_Apollo_Landscape
  }
  n72 = SetMaterialAttributes!(MaterialAttributes, Base Color=@n71.BaseColor, Specular=@n71.BaseColor, Roughness=@n71.BaseColor, Normal=@n71.BaseColor) {attributeSetTypes [
    #1
    #2
    #3
    #4
   ]}
  n73 = MaterialFunctionCall!(
   IsStaticSnow
   UseAlternateStaticSnowMask
   AlternateStaticSnowMask
   "UseRoad&&OcclusionMask"
   IsForGrassMeshToPushDown
   AdditionalMaskToRemoveSnow
   IsFallingSnowFX
   UseMinimapCoords
   IncreaseSnowUsingSlope
   UseSmoothBorder
   LerpHighM
   TemporalHeightLerpContrast
   IsLobbyIsland
   UseBR-IslandMask=@n70
   IsTree
   IsForRenderTarget
   UseSparkle
   UseSnowNormals=StaticBool() {value=true}
   UseAdditionalMaskToRemoveSnow
   IsTangentSpaceInput
   UseVertexNormalsForBasicSnow
   UseTangentSnowNormals=@n70
   Material=MaterialFunctionCall(NoRVT=@n16, RVT=@n72) {materialFunction=&FN_RVT_Switch_Material}.Result
  ) {materialFunction=/Game/Athena/Apollo/Environments/Landscape/Materials/Biome/UniqueBiomeMaterialFunctions/MF_Apollo_Snow_Progression}
  n74 = StaticSwitchParameter(
   True=MaterialFunctionCall!(In=@n73.Material, Mask=LandscapeLayerSample() {parameterName=BiomeD}, UseCheapSlopeMask, MaskReflectionCurveOnZ, HueShiftOverride) {materialFunction=/Game/TechArt/Seasonal/LackLuster/MaterialFunctions/MF_LackLuster_Landscape_PostRVT}.Result
   False=@n73.Material
  ) {parameterName=UseLackluster}
  SetMaterialAttributes!(
   MaterialAttributes=MaterialFunctionCall!(
    Material=MaterialFunctionCall!(
     In=@n74
     UseHeightFalloff
     OnFireEmissiveOverride
     HeightFalloffRadius
     HeightFalloffHardness
     VariationMasksMobile
     In_Grass
     Is_Stylised_WE
    ) {desc=Fire, materialFunction=/Game/Packages/Fortress_SharedMaterials/Base_Material_Functions/MF_Curie_Landscape}.Result
    DebugInputAsBasecolor
    Non-RVT material=NamedRerouteUsage()
    DebugMaterialAs1Color=VectorParameter() {parameterName=DebugBiomeColor, defaultValue=<rgba 1 0.0502 0.0203 1>}.RGB
   ) {materialFunction=/Game/Athena/Apollo/Environments/Landscape/Materials/Biome/CommonMaterials/MF_LS_DEBUG}.Result
   Opacity Mask=LandscapeVisibilityMask()
  ) {attributeSetTypes [#5]}
 ]
}
