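# Landscape material for the Artemis lobby island, layering common ground, Nightmare jungle-grass and mud biomes, and slope-based cliffs.
# The blend writes to a runtime virtual texture, with a mobile alternative, and the surface switches between sampling it and the mobile material.
# Layer masks drive chrome, dirt and stone physical materials and four grass types, with world-height cutoffs for beach and ocean.
# A visibility mask cuts landscape holes.
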Material /Game/Athena/Apollo/Environments/Landscape/Materials/Artemis/M_Artemis_LobbyIsland
graph {
 properties {
  bTangentSpaceNormal=false
  bUseMaterialAttributes=true
  bUsedWithVirtualHeightfieldMesh=true
  pixelDepthOffsetMode=PDOM_Legacy
  usedShadingModels=Default Lit
 }
 nodes [
  Add(A=Constant() {r=-0.5}, B=Multiply(A=Constant() {r=2}, B=LandscapeLayerSample() {desc=Flowers, parameterName=BiomeB}))
  n1 = NamedRerouteUsage()
  n2 = NamedRerouteUsage()
  n3 = Multiply(
   A=Max(A=LandscapeLayerSample() {desc=ForestFloor, parameterName=BiomeA}, B=LandscapeLayerSample() {parameterName=BiomeB})
   B=Constant() {r=3}
  )
  MaterialExpressionOneMinus_0 = OneMinus(None=Max(A=Max(A=@n1, B=@n2), B=Saturate(None=@n3)))
  MaterialExpressionSubtract_0 = Subtract(A=LandscapeLayerSample() {parameterName=BiomeA}, B=LandscapeLayerSample() {parameterName=BiomeB})
  n4 = ComponentMask(None=WorldPosition().XYZ) {b=true}
  n5 = Subtract(A=@n4, B=Constant() {r=-1200})
  n6 = Constant() {r=1}
  n7 = OneMinus(None=Saturate(None=Divide(A=@n5, B=@n6)))
  MaterialExpressionSubtract_16 = Subtract(A=Subtract(A=LandscapeLayerSample() {parameterName=Sand}, B=NamedRerouteUsage()), B=@n7)
  n8 = Divide(A=Subtract(A=@n4, B=Constant() {r=-700}), B=@n6)
  MaterialExpressionSubtract_8 = Subtract(A=OneMinus(None=Saturate(None=@n8)), B=@n7)
  LandscapeGrassOutput!(
   Grass_Standard=@MaterialExpressionOneMinus_0
   TropicalJungle=@MaterialExpressionSubtract_0
   Beach=@MaterialExpressionSubtract_16
   Ocean=@MaterialExpressionSubtract_8
  ) {grassTypes [
    Grass_Standard {
     grassType=/Game/Athena/Apollo/Environments/Landscape/SmallFoliage/DataAssets/LandscapeGrassType/LandscapeGrassType_StandardGrass_Lobby
     input {
      expression=&self:MaterialExpressionOneMinus_0
      inputName=None
      mask=0
      maskA=0
      maskB=0
      maskG=0
      maskR=0
      outputIndex=0
     }
    }
    TropicalJungle {
     grassType=/Game/Athena/Apollo/Environments/Landscape/SmallFoliage/DataAssets/LandscapeGrassType/LandscapeGrassType_Nightmare
     input {
      expression=&self:MaterialExpressionSubtract_0
      inputName=None
      mask=0
      maskA=0
      maskB=0
      maskG=0
      maskR=0
      outputIndex=0
     }
    }
    Beach {
     grassType=/Game/Athena/Apollo/Environments/Landscape/SmallFoliage/DataAssets/LandscapeGrassType/LandscapeGrassType_Beach
     input {
      expression=&self:MaterialExpressionSubtract_16
      inputName=None
      mask=0
      maskA=0
      maskB=0
      maskG=0
      maskR=0
      outputIndex=0
     }
    }
    Ocean {
     grassType=/Game/Athena/Apollo/Environments/Landscape/SmallFoliage/DataAssets/LandscapeGrassType/LandscapeGrassType_Ocean
     input {
      expression=&self:MaterialExpressionSubtract_8
      inputName=None
      mask=0
      maskA=0
      maskB=0
      maskG=0
      maskR=0
      outputIndex=0
     }
    }
   ]}
  MaterialExpressionMaterialFunctionCall_30 = MaterialFunctionCall!(SampleValue=LandscapeLayerSample() {parameterName=BiomeD}) {materialFunction=/Game/Athena/Apollo/Environments/Landscape/Materials/MaterialFunctions/MF_ArtemisChromePhysicalMaterialMask}
  n9 = LandscapeLayerSample() {parameterName=Gravel}
  n10 = LandscapeLayerSample() {parameterName=Mud}
  n11 = Max(A=Multiply(A=Max(A=@n9, B=@n10), B=Constant() {r=2}), B=Multiply(A=LandscapeLayerSample() {parameterName=Sand}, B=Constant() {r=5}))
  n12 = Ceil(None=LandscapeLayerSample() {parameterName=Asphalt})
  MaterialExpressionMax_12 = Max(A=NamedRerouteUsage(), B=Saturate(None=Max(A=@n11, B=@n12)))
  MaterialExpressionOneMinus_6 = OneMinus(None=NamedRerouteUsage())
  LandscapePhysicalMaterialOutput!(
   Chrome_PhysicalMaterial=@MaterialExpressionMaterialFunctionCall_30.Result
   Dirt=@MaterialExpressionMax_12
   Stone=@MaterialExpressionOneMinus_6
  ) {inputs [
    {
     physicalMaterial=/Game/TechArt/Seasonal/LackLuster/Physics/Chrome_PhysicalMaterial
     input {
      expression=&self:MaterialExpressionMaterialFunctionCall_30
      inputName=None
      mask=0
      maskA=0
      maskB=0
      maskG=0
      maskR=0
      outputIndex=0
     }
    }
    {
     physicalMaterial=/Game/FortressPhysicalMaterials/Dirt
     input {
      expression=&self:MaterialExpressionMax_12
      inputName=None
      mask=0
      maskA=0
      maskB=0
      maskG=0
      maskR=0
      outputIndex=0
     }
    }
    {
     physicalMaterial=/Game/FortressPhysicalMaterials/Stone
     input {
      expression=&self:MaterialExpressionOneMinus_6
      inputName=None
      mask=0
      maskA=0
      maskB=0
      maskG=0
      maskR=0
      outputIndex=0
     }
    }
   ]}
  n13 = StaticBool()
  n14 = GetMaterialAttributes(None=MaterialFunctionCall!(PutSandBeforeGravelOverride, UseSnow=StaticBool(), UseOceanLevel) {materialFunction=/Game/Athena/Apollo/Environments/Landscape/Materials/Biome/CommonMaterials/Mobile/MF_Mobile_CommonLandscapeLayers}.Material) {attributeGetTypes [#1]}
  MaterialExpressionConstant_13 = Constant()
  n15 = LandscapeLayerBlend!(
   Layer Gravel=@MaterialExpressionConstant_13
   Layer Mud=@MaterialExpressionConstant_13
   Layer Sand=@MaterialExpressionConstant_13
   Layer Asphalt=@MaterialExpressionConstant_13
  ) {layers [
    {
     blendType=LB_WeightBlend
     constHeightInput=0
     layerName=Gravel
     previewWeight=0
     constLayerInput=<0 0 0>
     heightInput {expression=null, inputName=None, mask=0, maskA=0, maskB=0, maskG=0, maskR=0, outputIndex=0}
     layerInput {
      expression=&self:MaterialExpressionConstant_13
      inputName=None
      mask=0
      maskA=0
      maskB=0
      maskG=0
      maskR=0
      outputIndex=0
     }
    }
    {
     blendType=LB_WeightBlend
     constHeightInput=0
     layerName=Mud
     previewWeight=0
     constLayerInput=<0 0 0>
     heightInput {expression=null, inputName=None, mask=0, maskA=0, maskB=0, maskG=0, maskR=0, outputIndex=0}
     layerInput {
      expression=&self:MaterialExpressionConstant_13
      inputName=None
      mask=0
      maskA=0
      maskB=0
      maskG=0
      maskR=0
      outputIndex=0
     }
    }
    {
     blendType=LB_WeightBlend
     constHeightInput=0
     layerName=Sand
     previewWeight=0
     constLayerInput=<0 0 0>
     heightInput {expression=null, inputName=None, mask=0, maskA=0, maskB=0, maskG=0, maskR=0, outputIndex=0}
     layerInput {
      expression=&self:MaterialExpressionConstant_13
      inputName=None
      mask=0
      maskA=0
      maskB=0
      maskG=0
      maskR=0
      outputIndex=0
     }
    }
    {
     blendType=LB_WeightBlend
     constHeightInput=0
     layerName=Asphalt
     previewWeight=0
     constLayerInput=<0 0 0>
     heightInput {expression=null, inputName=None, mask=0, maskA=0, maskB=0, maskG=0, maskR=0, outputIndex=0}
     layerInput {
      expression=&self:MaterialExpressionConstant_13
      inputName=None
      mask=0
      maskA=0
      maskB=0
      maskG=0
      maskR=0
      outputIndex=0
     }
    }
   ]}
  n16 = SetMaterialAttributes!(MaterialAttributes=@n14.MaterialAttributes, Base Color=Add(A=@n14.Base Color, B=Multiply(A=@n15) {constB=0})) {desc="This must remain in material,\r\nnot a function.", attributeSetTypes [#1]}
  n17 = RuntimeVirtualTextureSampleParameter() {
   group=RVT
   materialType=BaseColor_Normal_Specular_YCoCg
   parameterName=RVT Sample Target
   virtualTexture=/Game/Athena/Apollo/Maps/Landscape/VT/RVT_Apollo_LobbyIsland
  }
  n18 = SetMaterialAttributes!(MaterialAttributes, Base Color=@n17.BaseColor, Specular=@n17.BaseColor, Roughness=@n17.BaseColor, Normal=@n17.BaseColor) {attributeSetTypes [
    #1
    #2
    #3
    #4
   ]}
  n19 = MaterialFunctionCall(NoRVT=@n16, RVT=@n18) {materialFunction=&FN_RVT_Switch_Material}
  MaterialFunctionCall!(
   IsStaticSnow
   UseAlternateStaticSnowMask
   AlternateStaticSnowMask
   "UseRoad&&OcclusionMask"
   IsForGrassMeshToPushDown
   AdditionalMaskToRemoveSnow
   IsFallingSnowFX
   UseMinimapCoords
   IncreaseSnowUsingSlope
   UseSmoothBorder
   LerpHighM
   TemporalHeightLerpContrast
   IsLobbyIsland
   UseBR-IslandMask=@n13
   IsTree
   IsForRenderTarget
   UseSparkle
   UseSnowNormals=StaticBool() {value=true}
   UseAdditionalMaskToRemoveSnow
   IsTangentSpaceInput
   UseVertexNormalsForBasicSnow
   UseTangentSnowNormals=@n13
   Material=@n19.Result
  ) {materialFunction=/Game/Athena/Apollo/Environments/Landscape/Materials/Biome/UniqueBiomeMaterialFunctions/MF_Apollo_Snow_Progression}
  n20 = LandscapeLayerSample() {parameterName=Mud}
  n21 = LandscapeLayerSample() {parameterName=Sand}
  n22 = LandscapeLayerSample() {parameterName=Asphalt}
  Max(A=Max(A=Max(A=@n20, B=@n21), B=Ceil(None=@n22)), B=NamedRerouteUsage())
  n23 = MaterialFunctionCall!(
   AdditionalGrassBorderColorMask=LandscapeLayerSample() {desc=ForestFloor, parameterName=BiomeB}
   PutSandBeforeGravel=StaticBool() {value=true}
   UseOceanLevel
  ) {materialFunction=/Game/Athena/Apollo/Environments/Landscape/Materials/Biome/CommonMaterials/MF_CommonLandscapeLayers}
  NamedRerouteDeclaration(None=@n23.GrassNoiseMask) {name=GrassNoiseMask, nodeColor=<rgba 1 0.0234 0 1>}
  NamedRerouteDeclaration(None=@n23.GrassBaseColorDetail) {name=GrassBaseColorDetail, nodeColor=<rgba 0 1 0.625 1>}
  NamedRerouteDeclaration(None=@n23.Material) {name=CommonMaterial, nodeColor=<rgba 1 0 0.6094 1>}
  NamedRerouteDeclaration(None=@n23.CommonLayerMaskForGrassFoliage) {name=GM-CommonLayers, nodeColor=<rgba 0 0.1562 1 1>}
  NamedRerouteDeclaration(None=@n23.GrassHeightmap) {name=GrassHeightmap, nodeColor=<rgba 0.8984 1 0 1>}
  NamedRerouteDeclaration(None=@n23.GravelMaskForGrassFoliage) {name=GM-Gravel, nodeColor=<rgba 0.3594 1 0 1>}
  n24 = MaterialFunctionCall!(
   Scale=ScalarParameter() {defaultValue=1.5, parameterName=BiomeA-Scale}
   Diffuse=TextureObject() {texture=/Game/Environments/Apollo/Sets/Nightmare/Textures/T_Nightmare_JungleGrass_D_01}
   Normal=TextureObject() {samplerType=SAMPLERTYPE_Normal, texture=/Game/Environments/Apollo/Sets/Nightmare/Textures/T_Nightmare_JungleGrass_N_01}
   S_Mask=TextureObject() {samplerType=SAMPLERTYPE_Masks, texture=/Game/Environments/Apollo/Sets/Nightmare/Textures/T_Nightmare_JungleGrass_S_01}
  ) {materialFunction=/Game/Athena/Apollo/Environments/Landscape/Materials/Biome/UniqueBiomeMaterialFunctions/MF_Biome_Generic}
  n25 = MaterialFunctionCall!(
   Material 1=@n23.Material
   Height Texture 1=@n23.Material
   Material 2=@n24.Material
   Height Texture 2=NamedRerouteDeclaration(None=@n24.Height) {name=BiomeA-ForestFloorHeight, nodeColor=<rgba 0 1 0.4375 1>}
   Transistion Phase=LandscapeLayerSample() {parameterName=BiomeA}
   UseContrast
   Contrast
   AddLayerMask
  ) {materialFunction=/Game/Athena/Apollo/Environments/Landscape/Materials/Biome/CommonMaterials/MF_BlendLandscapeLayers}
  n26 = MaterialFunctionCall(
   Scale=ScalarParameter() {defaultValue=1.5, parameterName=BiomeB-Scale}
   Diffuse=TextureObject() {texture=/Game/Environments/Apollo/Sets/Nightmare/Textures/T_Nightmare_Mud_D_01}
   Normal=TextureObject() {samplerType=SAMPLERTYPE_Normal, texture=/Game/Environments/Apollo/Sets/Nightmare/Textures/T_Nightmare_Mud_N_01}
   S_Mask=TextureObject() {samplerType=SAMPLERTYPE_Masks, texture=/Game/Environments/Apollo/Sets/Nightmare/Textures/T_Nightmare_Mud_S_01}
  ) {materialFunction=&MF_Biome_Generic}
  n27 = MaterialFunctionCall(
   Material 1=@n25.BlendedMaterial
   Height Texture 1=@n25.BlendedMaterial
   Material 2=@n26.Material
   Height Texture 2=@n26.Material
   Transistion Phase=LandscapeLayerSample() {parameterName=BiomeB}
  ) {materialFunction=&MF_BlendLandscapeLayers}
  n28 = MaterialFunctionCall!(
   Material=@n27.BlendedMaterial
   Heightmap=@n27.BlendedMaterial
   NewSlopeMaskMask
   Displacement
   AsphaltLayer=LandscapeLayerSample() {parameterName=Sand}
   UseAlternateCliffMat
   AltCliffHeight
   AltCliffMat
  ) {materialFunction=/Game/Athena/Apollo/Environments/Landscape/Materials/Biome/CommonMaterials/MF_SlopeBlendCliff}
  NamedRerouteDeclaration(None=@n28.SlopeMask) {name=SlopeMask, nodeColor=<rgba 1 1 0 1>}
  n29 = LinearInterpolate(
   A=ScalarParameter() {defaultValue=-0.5, parameterName=BiomeA-FFSpawnLa}
   B=ScalarParameter() {defaultValue=1.5, parameterName=BiomeA-FFSpawnLb}
   Alpha=LandscapeLayerSample() {desc=ForestFloor, parameterName=BiomeA}
  )
  NamedRerouteDeclaration(None=MaterialFunctionCall(
    Transition Phase=Saturate(None=@n29)
    Height Texture=NamedRerouteUsage()
    Contrast=ScalarParameter() {defaultValue=0.5, parameterName=BiomeA-FFSpawnCont}
   ) {desc=ForestFloor, materialFunction=&HeightLerp}.Alpha) {name=GM-ForestFloor, nodeColor=<rgba 0.2891 0 1 1>}
  NamedRerouteDeclaration(None=@n28.CliffMask) {name=CliffMask, nodeColor=<rgba 1 0 0.75 1>}
  NamedRerouteDeclaration(None=@n28.LandscapeGrassMeshMask) {name=GM-Cliff, nodeColor=<rgba 0 1 0.9531 1>}
  n30 = MaterialFunctionCall!(Material=@n28.Material) {materialFunction=/Game/Athena/Apollo/Maps/Landscape/VT/MaterialFunctions/TransformTangentToWorld}
  n31 = GetMaterialAttributes(None=MaterialFunctionCall!(MobileMaterial=@n16, HighEndMaterial=@n30.Result) {materialFunction=/Game/Athena/Asteria/Environments/Landscape/Materials/MaterialFunctions/FN_RvtWriteSwitch}.Result) {attributeGetTypes [
    #1
    #2
    #3
    #4
   ]}
  RuntimeVirtualTextureOutput!(
   BaseColor=MaterialFunctionCall!(BaseColor=@n31.Base Color) {materialFunction=/Game/Athena/Apollo/Environments/Landscape/Materials/Biome/CommonMaterials/MF_ReduceRTV-CompressionArtifacts}.BaseColor
   Specular=@n31.Specular
   Roughness=@n31.Specular
   Normal=@n31.Specular
   WorldHeight=ComponentMask(None=WorldPosition().XYZ) {b=true}
   Opacity
   Mask
   Displacement
   Mask4
  )
  ScalarParameter() {defaultValue=1, parameterName=BiomeB-cont}
  ScalarParameter() {parameterName=BiomeA-cont}
  SetMaterialAttributes!(
   MaterialAttributes=MaterialFunctionCall!(
    Material=@n19.Result
    DebugInputAsBasecolor
    Non-RVT material
    DebugMaterialAs1Color=VectorParameter() {parameterName=DebugBiomeColor, defaultValue=<rgba 1 0.0502 0.0203 1>}.RGB
   ) {materialFunction=/Game/Athena/Apollo/Environments/Landscape/Materials/Biome/CommonMaterials/MF_LS_DEBUG}.Result
   Opacity Mask=LandscapeVisibilityMask()
  ) {attributeSetTypes [#5]}
  StaticBool() {value=true}
  StaticBool() {value=true}
  StaticSwitchParameter(
   True=MaterialFunctionCall!(RVT=@n18, ByPass RVT=@n30.Result) {materialFunction=/Game/Athena/Apollo/Maps/Landscape/VT/MaterialFunctions/EnableRVT}.Result
   False=@n30.Result
  ) {defaultValue=true, group=RVT Base, parameterName=RVT Enable}
 ]
}
