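# Landscape master material for the Artemis mountains biome.
# Common layers height-blend with ForestFloor, Heather and Scree and slope-blended cliffs, then feed a runtime virtual texture, with a simpler mobile path.
# The RVT sample receives snow progression, optional LackLuster treatment masked by BiomeD, fire and debug functions.
# Layer, cliff and height masks also drive landscape grass and physical material outputs; grass data bakes into levels on save.
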Material /Game/Athena/Apollo/Environments/Landscape/Materials/Artemis/M_Artemis_Mountains
graph {
 properties {bTangentSpaceNormal=false, bUseMaterialAttributes=true, pixelDepthOffsetMode=PDOM_Legacy, usedShadingModels=Default Lit}
 nodes [
  Constant() {r=-0.005}
  Constant() {r=0.004}
  Constant() {r=0.02}
  Constant() {r=-0.04}
  Constant() {r=0.04}
  Constant() {r=0.015}
  Constant() {r=0.0075}
  n1 = NamedRerouteUsage()
  MaterialExpressionMaterialFunctionCall_2 = MaterialFunctionCall!(In=Subtract(A=@n1, B=LandscapeLayerSample() {parameterName=BiomeD})) {materialFunction=/Game/Athena/Apollo/Environments/Landscape/Materials/Biome/CommonMaterials/MF_LS_GrassSwitch}
  MaterialExpressionMaterialFunctionCall_12 = MaterialFunctionCall(In=@n1) {materialFunction=&MF_LS_GrassSwitch}
  MaterialExpressionMaterialFunctionCall_20 = MaterialFunctionCall() {materialFunction=&MF_LS_GrassSwitch}
  MaterialExpressionMaterialFunctionCall_21 = MaterialFunctionCall(In=NamedRerouteUsage()) {materialFunction=&MF_LS_GrassSwitch}
  MaterialExpressionMaterialFunctionCall_22 = MaterialFunctionCall(In=Subtract(A=NamedRerouteUsage(), B=LandscapeLayerSample() {parameterName=BiomeD})) {materialFunction=&MF_LS_GrassSwitch}
  n2 = LandscapeLayerSample() {parameterName=BiomeD}
  MaterialExpressionMaterialFunctionCall_23 = MaterialFunctionCall(In=Subtract(A=NamedRerouteUsage(), B=@n2)) {materialFunction=&MF_LS_GrassSwitch}
  MaterialExpressionMaterialFunctionCall_24 = MaterialFunctionCall(In=Subtract(A=NamedRerouteUsage(), B=@n2)) {materialFunction=&MF_LS_GrassSwitch}
  MaterialExpressionMaterialFunctionCall_25 = MaterialFunctionCall(In=Subtract(A=NamedRerouteUsage(), B=@n2)) {materialFunction=&MF_LS_GrassSwitch}
  MaterialExpressionMaterialFunctionCall_26 = MaterialFunctionCall() {materialFunction=&MF_LS_GrassSwitch}
  MaterialExpressionMaterialFunctionCall_27 = MaterialFunctionCall() {materialFunction=&MF_LS_GrassSwitch}
  MaterialExpressionMaterialFunctionCall_28 = MaterialFunctionCall() {materialFunction=&MF_LS_GrassSwitch}
  MaterialExpressionMaterialFunctionCall_29 = MaterialFunctionCall() {materialFunction=&MF_LS_GrassSwitch}
  MaterialExpressionMaterialFunctionCall_7 = MaterialFunctionCall() {materialFunction=&MF_LS_GrassSwitch}
  LandscapeGrassOutput!(
   Grass_Standard=@MaterialExpressionMaterialFunctionCall_2.Result
   Bush=@MaterialExpressionMaterialFunctionCall_12.Result
   GrasslandFlowers=@MaterialExpressionMaterialFunctionCall_20.Result
   Scree=@MaterialExpressionMaterialFunctionCall_21.Result
   Scree_Grass=@MaterialExpressionMaterialFunctionCall_22.Result
   Heather=@MaterialExpressionMaterialFunctionCall_23.Result
   ForestFloor=@MaterialExpressionMaterialFunctionCall_24.Result
   Beach=@MaterialExpressionMaterialFunctionCall_25.Result
   Ocean=@MaterialExpressionMaterialFunctionCall_26.Result
   DesertPlants=@MaterialExpressionMaterialFunctionCall_27.Result
   DesertScree=@MaterialExpressionMaterialFunctionCall_28.Result
   TropicalJungle=@MaterialExpressionMaterialFunctionCall_29.Result
   FrontierPlants=@MaterialExpressionMaterialFunctionCall_7.Result
  ) {
   desc="This must have the exact inputs in every biome material!\r\nThis data is baked into the level data when the level is\r\nsaved, so any changes to this require the affected level\r\nto be saved && submitted."
   grassTypes [
    Grass_Standard {
     grassType=/Game/Athena/Apollo/Environments/Landscape/SmallFoliage/DataAssets/LandscapeGrassType/LandscapeGrassType_StandardGrass
     input {
      expression=&self:MaterialExpressionMaterialFunctionCall_2
      inputName=None
      mask=0
      maskA=0
      maskB=0
      maskG=0
      maskR=0
      outputIndex=0
     }
    }
    Bush {
     grassType=/Game/Athena/Environments/Landscape/Foliage/Athena_Foliage_Bushes
     input {
      expression=&self:MaterialExpressionMaterialFunctionCall_12
      inputName=None
      mask=0
      maskA=0
      maskB=0
      maskG=0
      maskR=0
      outputIndex=0
     }
    }
    GrasslandFlowers {
     grassType=/Game/Athena/Apollo/Environments/Landscape/SmallFoliage/DataAssets/LandscapeGrassType/LandscapeGrassType_FarmFlowers
     input {
      expression=&self:MaterialExpressionMaterialFunctionCall_20
      inputName=None
      mask=0
      maskA=0
      maskB=0
      maskG=0
      maskR=0
      outputIndex=0
     }
    }
    Scree {
     grassType=/Game/Athena/Apollo/Environments/Landscape/SmallFoliage/DataAssets/LandscapeGrassType/LandscapeGrassType_Scree
     input {
      expression=&self:MaterialExpressionMaterialFunctionCall_21
      inputName=None
      mask=0
      maskA=0
      maskB=0
      maskG=0
      maskR=0
      outputIndex=0
     }
    }
    Scree_Grass {
     grassType=/Game/Athena/Apollo/Environments/Landscape/SmallFoliage/DataAssets/LandscapeGrassType/LandscapeGrassType_ScreeGrass
     input {
      expression=&self:MaterialExpressionMaterialFunctionCall_22
      inputName=None
      mask=0
      maskA=0
      maskB=0
      maskG=0
      maskR=0
      outputIndex=0
     }
    }
    Heather {
     grassType=/Game/Athena/Apollo/Environments/Landscape/SmallFoliage/DataAssets/LandscapeGrassType/LandscapeGrassType_Heather
     input {
      expression=&self:MaterialExpressionMaterialFunctionCall_23
      inputName=None
      mask=0
      maskA=0
      maskB=0
      maskG=0
      maskR=0
      outputIndex=0
     }
    }
    ForestFloor {
     grassType=/Game/Athena/Apollo/Environments/Landscape/SmallFoliage/DataAssets/LandscapeGrassType/LandscapeGrassType_ForestGrass
     input {
      expression=&self:MaterialExpressionMaterialFunctionCall_24
      inputName=None
      mask=0
      maskA=0
      maskB=0
      maskG=0
      maskR=0
      outputIndex=0
     }
    }
    Beach {
     grassType=/Game/Athena/Apollo/Environments/Landscape/SmallFoliage/DataAssets/LandscapeGrassType/LandscapeGrassType_Beach
     input {
      expression=&self:MaterialExpressionMaterialFunctionCall_25
      inputName=None
      mask=0
      maskA=0
      maskB=0
      maskG=0
      maskR=0
      outputIndex=0
     }
    }
    Ocean {
     grassType=/Game/Athena/Apollo/Environments/Landscape/SmallFoliage/DataAssets/LandscapeGrassType/LandscapeGrassType_Ocean
     input {
      expression=&self:MaterialExpressionMaterialFunctionCall_26
      inputName=None
      mask=0
      maskA=0
      maskB=0
      maskG=0
      maskR=0
      outputIndex=0
     }
    }
    DesertPlants {
     grassType=/Game/Athena/Apollo/Environments/Landscape/SmallFoliage/DataAssets/LandscapeGrassType/LandscapeGrassType_DesertPlants
     input {
      expression=&self:MaterialExpressionMaterialFunctionCall_27
      inputName=None
      mask=0
      maskA=0
      maskB=0
      maskG=0
      maskR=0
      outputIndex=0
     }
    }
    DesertScree {
     grassType=/Game/Athena/Apollo/Environments/Landscape/SmallFoliage/DataAssets/LandscapeGrassType/LandscapeGrassType_DesertScree
     input {
      expression=&self:MaterialExpressionMaterialFunctionCall_28
      inputName=None
      mask=0
      maskA=0
      maskB=0
      maskG=0
      maskR=0
      outputIndex=0
     }
    }
    TropicalJungle {
     grassType=/Game/Athena/Apollo/Environments/Landscape/SmallFoliage/DataAssets/LandscapeGrassType/LandscapeGrassType_Nightmare
     input {
      expression=&self:MaterialExpressionMaterialFunctionCall_29
      inputName=None
      mask=0
      maskA=0
      maskB=0
      maskG=0
      maskR=0
      outputIndex=0
     }
    }
    FrontierPlants {
     grassType=/Game/Athena/Apollo/Environments/Landscape/SmallFoliage/DataAssets/LandscapeGrassType/NewLandscapeGrassType_FrontierGrass
     input {
      expression=&self:MaterialExpressionMaterialFunctionCall_7
      inputName=None
      mask=0
      maskA=0
      maskB=0
      maskG=0
      maskR=0
      outputIndex=0
     }
    }
   ]
  }
  MaterialExpressionMaterialFunctionCall_32 = MaterialFunctionCall!(SampleValue=LandscapeLayerSample() {parameterName=BiomeD}) {materialFunction=/Game/Athena/Apollo/Environments/Landscape/Materials/MaterialFunctions/MF_ArtemisChromePhysicalMaterialMask}
  n3 = NamedRerouteUsage()
  n4 = OneMinus(None=@n3) {desc="White=Cliffs"}
  n5 = LandscapeLayerSample() {parameterName=Gravel}
  n6 = LandscapeLayerSample() {parameterName=Mud}
  n7 = Saturate(None=Multiply(A=Max(A=@n5, B=@n6), B=Constant() {r=2}))
  n8 = Subtract(A=@n7, B=@n4) {desc="Remove Cliff from Gravel,Mud, Sand, Snow"}
  n9 = LandscapeLayerSample() {parameterName=Sand}
  n10 = Saturate(None=Multiply(A=LandscapeLayerSample() {parameterName=Asphalt}, B=Constant() {r=3}))
  n11 = Max(A=Max(A=NamedRerouteUsage(), B=LandscapeLayerSample() {desc=Heather, parameterName=BiomeB}), B=@n10)
  MaterialExpressionSaturate_12 = Saturate(None=Subtract(A=Saturate(None=@n8), B=Max(A=@n9, B=@n11)) {desc="Remove ForestFloor, Heather, Asphalt, Sand"}) {desc=Final Dirt mask}
  MaterialExpressionReroute_51 = Reroute(None=@n4) {desc=Final Stone mask}
  n12 = Subtract(A=@n3, B=Max(A=@n9, B=@n7)) {desc="Remove Gravel,Mud, Sand, Snow"}
  MaterialExpressionSaturate_14 = Saturate(None=Subtract(A=Saturate(None=@n12), B=@n10) {desc=Remove Asphalt}) {desc=Final Grass mask}
  MaterialExpressionReroute_61 = Reroute(None=@n10)
  n13 = Saturate(None=Subtract(A=@n9, B=@n4) {desc=Remove Cliff})
  MaterialExpressionReroute_7 = Reroute(None=Saturate(None=Subtract(A=@n13, B=@n11) {desc="Remove ForestFloor, Flowers, Asphalt"}))
  LandscapePhysicalMaterialOutput!(
   Chrome_PhysicalMaterial=@MaterialExpressionMaterialFunctionCall_32.Result
   Dirt=@MaterialExpressionSaturate_12
   Stone=@MaterialExpressionReroute_51
   Grass=@MaterialExpressionSaturate_14
   Asphalt=@MaterialExpressionReroute_61
   Sand=@MaterialExpressionReroute_7
  ) {
   desc="These are layered in order of importance.\r\nIf Dirt=1 and Stone=1, then Dirt will \"win\".\r\nTheoretically, these values aren't clamped.\r\nSo a value of 2 should trump a value of 1."
   inputs [
    {
     physicalMaterial=/Game/TechArt/Seasonal/LackLuster/Physics/Chrome_PhysicalMaterial
     input {
      expression=&self:MaterialExpressionMaterialFunctionCall_32
      inputName=None
      mask=0
      maskA=0
      maskB=0
      maskG=0
      maskR=0
      outputIndex=0
     }
    }
    {
     physicalMaterial=/Game/FortressPhysicalMaterials/Dirt
     input {
      expression=&self:MaterialExpressionSaturate_12
      inputName=None
      mask=0
      maskA=0
      maskB=0
      maskG=0
      maskR=0
      outputIndex=0
     }
    }
    {
     physicalMaterial=/Game/FortressPhysicalMaterials/Stone
     input {
      expression=&self:MaterialExpressionReroute_51
      inputName=None
      mask=0
      maskA=0
      maskB=0
      maskG=0
      maskR=0
      outputIndex=0
     }
    }
    {
     physicalMaterial=/Game/FortressPhysicalMaterials/Grass
     input {
      expression=&self:MaterialExpressionSaturate_14
      inputName=None
      mask=0
      maskA=0
      maskB=0
      maskG=0
      maskR=0
      outputIndex=0
     }
    }
    {
     physicalMaterial=/Game/FortressPhysicalMaterials/Asphalt
     input {
      expression=&self:MaterialExpressionReroute_61
      inputName=None
      mask=0
      maskA=0
      maskB=0
      maskG=0
      maskR=0
      outputIndex=0
     }
    }
    {
     physicalMaterial=/Game/FortressPhysicalMaterials/Sand
     input {
      expression=&self:MaterialExpressionReroute_7
      inputName=None
      mask=0
      maskA=0
      maskB=0
      maskG=0
      maskR=0
      outputIndex=0
     }
    }
   ]
  }
  n14 = LinearInterpolate(A=@MaterialExpressionSaturate_12, B=@MaterialExpressionReroute_51, Alpha=ScalarParameter() {parameterName=DEBUG-PhysMat1})
  n15 = ScalarParameter() {parameterName=DEBUG-PhysMat2}
  LinearInterpolate(
   A=LinearInterpolate(
    A=LinearInterpolate(A=@n14, B=@MaterialExpressionSaturate_14, Alpha=@n15)
    B=@MaterialExpressionReroute_61
    Alpha=ScalarParameter() {parameterName=DEBUG-PhysMat3}
   )
   B=@MaterialExpressionReroute_7
   Alpha=ScalarParameter() {defaultValue=1, parameterName=DEBUG-PhysMat4}
  )
  n16 = GetMaterialAttributes(None=MaterialFunctionCall!(PutSandBeforeGravelOverride, UseSnow=NamedRerouteUsage(), UseOceanLevel) {materialFunction=/Game/Athena/Apollo/Environments/Landscape/Materials/Biome/CommonMaterials/Mobile/MF_Mobile_CommonLandscapeLayers}.Material) {attributeGetTypes [#1]}
  MaterialExpressionConstant_0 = Constant()
  n17 = LandscapeLayerBlend!(
   Layer Gravel=@MaterialExpressionConstant_0
   Layer Mud=@MaterialExpressionConstant_0
   Layer Sand=@MaterialExpressionConstant_0
   Layer Asphalt=@MaterialExpressionConstant_0
  ) {layers [
    {
     blendType=LB_WeightBlend
     constHeightInput=0
     layerName=Gravel
     previewWeight=0
     constLayerInput=<0 0 0>
     heightInput {expression=null, inputName=None, mask=0, maskA=0, maskB=0, maskG=0, maskR=0, outputIndex=0}
     layerInput {
      expression=&self:MaterialExpressionConstant_0
      inputName=None
      mask=0
      maskA=0
      maskB=0
      maskG=0
      maskR=0
      outputIndex=0
     }
    }
    {
     blendType=LB_WeightBlend
     constHeightInput=0
     layerName=Mud
     previewWeight=0
     constLayerInput=<0 0 0>
     heightInput {expression=null, inputName=None, mask=0, maskA=0, maskB=0, maskG=0, maskR=0, outputIndex=0}
     layerInput {
      expression=&self:MaterialExpressionConstant_0
      inputName=None
      mask=0
      maskA=0
      maskB=0
      maskG=0
      maskR=0
      outputIndex=0
     }
    }
    {
     blendType=LB_WeightBlend
     constHeightInput=0
     layerName=Sand
     previewWeight=0
     constLayerInput=<0 0 0>
     heightInput {expression=null, inputName=None, mask=0, maskA=0, maskB=0, maskG=0, maskR=0, outputIndex=0}
     layerInput {
      expression=&self:MaterialExpressionConstant_0
      inputName=None
      mask=0
      maskA=0
      maskB=0
      maskG=0
      maskR=0
      outputIndex=0
     }
    }
    {
     blendType=LB_WeightBlend
     constHeightInput=0
     layerName=Asphalt
     previewWeight=0
     constLayerInput=<0 0 0>
     heightInput {expression=null, inputName=None, mask=0, maskA=0, maskB=0, maskG=0, maskR=0, outputIndex=0}
     layerInput {
      expression=&self:MaterialExpressionConstant_0
      inputName=None
      mask=0
      maskA=0
      maskB=0
      maskG=0
      maskR=0
      outputIndex=0
     }
    }
   ]}
  n18 = SetMaterialAttributes!(MaterialAttributes=@n16.MaterialAttributes, Base Color=Add(A=@n16.Base Color, B=Multiply(A=@n17) {constB=0})) {desc="This must remain in material,\r\nnot a function.", attributeSetTypes [#1]}
  n19 = MaterialFunctionCall!(
   AdditionalGrassBorderColorMask=LandscapeLayerSample() {desc=ForestFloor, parameterName=BiomeA}
   PutSandBeforeGravel=StaticBool() {value=true}
   UseOceanLevel
  ) {materialFunction=/Game/Athena/Apollo/Environments/Landscape/Materials/Biome/CommonMaterials/MF_CommonLandscapeLayers}
  n20 = MaterialFunctionCall() {materialFunction=/Game/Athena/Apollo/Environments/Landscape/Materials/Biome/UniqueBiomeMaterialFunctions/MF_Biome_Forest}
  n21 = MaterialFunctionCall!(
   Material 1=@n19.Material
   Height Texture 1=@n19.Material
   Material 2=@n20.Material
   Height Texture 2=NamedRerouteDeclaration(None=@n20.Height) {name=BiomeA-ForestFloorHeight, nodeColor=<rgba 0.5 0 1 1>}
   Transistion Phase=LandscapeLayerSample() {desc=ForestFloor, parameterName=BiomeA}
   UseContrast
   Contrast
   AddLayerMask
  ) {materialFunction=/Game/Athena/Apollo/Environments/Landscape/Materials/Biome/CommonMaterials/MF_BlendLandscapeLayers}
  n22 = MaterialFunctionCall!(GrassHeightmap=NamedRerouteUsage(), GrassNoiseMask=NamedRerouteUsage(), GrassMaterial=NamedRerouteUsage()) {materialFunction=/Game/Athena/Apollo/Environments/Landscape/Materials/Biome/UniqueBiomeMaterialFunctions/MF_Biome_Heather}
  n23 = MaterialFunctionCall(
   Material 1=@n21.BlendedMaterial
   Height Texture 1=@n21.BlendedMaterial
   Material 2=@n22.Material
   Height Texture 2=@n22.Material
   Transistion Phase=LandscapeLayerSample() {desc=Heather, parameterName=BiomeB}
  ) {materialFunction=&MF_BlendLandscapeLayers}
  n24 = NamedRerouteUsage()
  n25 = LandscapeLayerSample() {desc=Scree, parameterName=BiomeC}
  n26 = MaterialFunctionCall!(LayerWeight=@n25) {materialFunction=/Game/Athena/Apollo/Environments/Landscape/Materials/Biome/UniqueBiomeMaterialFunctions/MF_Scree}
  n27 = MaterialFunctionCall(
   Material 1=StaticSwitch(True=@n21.BlendedMaterial, False=@n23.BlendedMaterial, Value=@n24)
   Height Texture 1=StaticSwitch(True=@n21.Resulting Height Map, False=@n23.Resulting Height Map, Value=@n24)
   Material 2=@n26.Material
   Height Texture 2=NamedRerouteDeclaration(None=@n26.Height) {name=MF_Scree-Height, nodeColor=<rgba 0.5 0 1 1>}
   Transistion Phase=@n25
   UseContrast=StaticBool() {value=true}
   Contrast=Constant() {r=0.5}
  ) {materialFunction=&MF_BlendLandscapeLayers}
  n28 = MaterialFunctionCall!(
   Material=@n27.BlendedMaterial
   Heightmap=@n27.BlendedMaterial
   NewSlopeMaskMask
   Displacement
   AsphaltLayer=LandscapeLayerSample() {parameterName=Sand}
   UseAlternateCliffMat
   AltCliffHeight
   AltCliffMat
  ) {materialFunction=/Game/Athena/Apollo/Environments/Landscape/Materials/Biome/CommonMaterials/MF_SlopeBlendCliff}
  n29 = MaterialFunctionCall!(Material=@n28.Material) {materialFunction=/Game/Athena/Apollo/Maps/Landscape/VT/MaterialFunctions/TransformTangentToWorld}
  n30 = GetMaterialAttributes(None=@n29.Result) {attributeGetTypes [#1]}
  n31 = NamedRerouteUsage()
  n32 = NamedRerouteUsage()
  n33 = StaticSwitch(
   True=SetMaterialAttributes(MaterialAttributes=@n30.MaterialAttributes, Base Color=LinearInterpolate(A=@n30.Base Color, B=@n31, Alpha=@n32)) {attributeSetTypes [#1]}
   False=@n29.Result
   Value=NamedRerouteUsage()
  )
  n34 = StaticSwitchParameter(
   True=MaterialFunctionCall!(
    MA=@n33
    Mask=LandscapeLayerSample() {parameterName=BiomeD}
    HeightIn=@n28.RegularHeightmap
    SlopeMask=@n28.RegularHeightmap
    BlendLandscapeLayers
   ) {materialFunction=/Game/TechArt/Seasonal/LackLuster/MaterialFunctions/MF_LackLuster_Landscape_RVT}.Result
   False=@n33
  ) {parameterName=UseLackluster}
  n35 = GetMaterialAttributes(None=MaterialFunctionCall!(MobileMaterial=@n18, HighEndMaterial=NamedRerouteDeclaration(None=@n34) {name=MaterialForRVT, nodeColor=<rgba 0.5 0 1 1>}) {materialFunction=/Game/Athena/Asteria/Environments/Landscape/Materials/MaterialFunctions/FN_RvtWriteSwitch}.Result) {attributeGetTypes [
    #1
    #2
    #3
    #4
   ]}
  n36 = ScalarParameter() {defaultValue=-0.04, parameterName=Test-rough-noise-compression}
  n37 = ScalarParameter() {defaultValue=0.05, parameterName=Test-rough-noise-compressionB}
  n38 = WorldPosition()
  n39 = ScalarParameter() {defaultValue=0.01, parameterName=Test-rough-noise-tile}
  n40 = TextureSample(UVs=ComponentMask(None=Multiply(A=@n38.XYZ, B=@n39)) {g=true, r=true}) {samplerSource=SSM_Wrap_WorldGroupSettings, samplerType=SAMPLERTYPE_LinearColor, texture=/Game/Global/Textures/Default/Noise/T_Apollo_TCNoise}
  LinearInterpolate(
   A=@n35.Specular
   B=Add(A=@n35.Specular, B=LinearInterpolate(A=@n36, B=@n37, Alpha=@n40.R))
   Alpha=ScalarParameter() {defaultValue=1, parameterName=greituh34etgajhhkiii}
  )
  Max(A=Ceil(None=NamedRerouteUsage()), B=@n7) {desc=OLD: Add Snow}
  Multiply(A=@n40.R, B=@n36) {constB=0.02}
  NamedRerouteDeclaration(None=@n19.GrassNoiseMask) {name=GrassNoiseMask, nodeColor=<rgba 0.5 0 1 1>}
  NamedRerouteDeclaration(None=@n19.Material) {name=CommonMaterial, nodeColor=<rgba 0.5 0 1 1>}
  n41 = VertexNormalWS()
  n42 = Saturate(None=LinearInterpolate(A=Constant() {r=-4}, B=Constant() {r=1.6832}, Alpha=ComponentMask(None=@n41) {b=true}))
  n43 = LandscapeLayerSample() {desc=Scree, parameterName=BiomeC}
  n44 = Constant() {r=2}
  n45 = Saturate(None=Add(A=Multiply(A=@n43, B=@n44), B=Constant() {r=-0.5}))
  NamedRerouteDeclaration(None=Max(A=NamedRerouteUsage(), B=Multiply(A=@n42, B=@n45))) {name=GM-Scree, nodeColor=<rgba 0.5 0 1 1>}
  n46 = LandscapeLayerSample() {desc=Heather, parameterName=BiomeB}
  n47 = Constant() {r=2}
  n48 = Saturate(None=Add(A=Multiply(A=@n46, B=@n47), B=Constant() {r=-0.5}))
  n49 = OneMinus(None=NamedRerouteUsage())
  NamedRerouteDeclaration(None=StaticSwitch(True=Constant(), False=Multiply(A=@n48, B=@n49), Value=NamedRerouteUsage())) {name=GM-Heather, nodeColor=<rgba 0.5 0 1 1>}
  n50 = NamedRerouteUsage()
  n51 = NamedRerouteUsage()
  n52 = Constant() {r=-0.5}
  n53 = Constant() {r=3}
  n54 = LandscapeLayerSample() {desc=Scree, parameterName=BiomeC}
  n55 = MaterialFunctionCall(Transition Phase=Saturate(None=LinearInterpolate(A=@n52, B=@n53, Alpha=@n54)), Height Texture=NamedRerouteUsage(), Contrast=Constant() {r=5}) {desc=Scree Masked, materialFunction=&HeightLerp}
  n56 = Max(A=Max(A=Max(A=@n50, B=@n51), B=Max(A=@n55.Alpha, B=@n54)), B=NamedRerouteUsage())
  n57 = OneMinus(None=@n56)
  n58 = Constant() {r=6}
  n59 = NamedRerouteUsage()
  n60 = OneMinus(None=Saturate(None=Multiply(A=@n58, B=@n59)))
  n61 = OneMinus(None=Max(A=@n56, B=LandscapeLayerSample() {desc=Heather, parameterName=BiomeB}))
  NamedRerouteDeclaration(None=StaticSwitch(True=Multiply(A=@n57, B=@n60), False=@n61, Value=NamedRerouteUsage())) {name=GM-Grass, nodeColor=<rgba 0.5 0 1 1>}
  n62 = Saturate(None=Multiply(A=Constant() {r=3}, B=LandscapeLayerSample() {desc=Scree, parameterName=BiomeC}))
  n63 = NamedRerouteUsage()
  n64 = NamedRerouteUsage()
  n65 = Constant() {r=20}
  n66 = Subtract(A=Subtract(A=Subtract(A=@n62, B=@n63), B=NamedRerouteUsage()), B=Saturate(None=Multiply(A=@n64, B=@n65)))
  n67 = Constant() {r=6}
  n68 = NamedRerouteUsage()
  n69 = OneMinus(None=Saturate(None=Multiply(A=@n67, B=@n68)))
  NamedRerouteDeclaration(None=StaticSwitch(True=Multiply(A=@n66, B=@n69), False=@n66, Value=NamedRerouteUsage())) {name=GM-ScreeGrass, nodeColor=<rgba 0.5 0 1 1>}
  n70 = LandscapeLayerSample() {parameterName=Sand}
  n71 = Max(A=NamedRerouteUsage(), B=NamedRerouteUsage())
  n72 = ComponentMask(None=WorldPosition().XYZ) {b=true}
  n73 = Constant() {r=200}
  n74 = Subtract(A=Saturate(None=Subtract(A=@n70, B=@n71)), B=Saturate(None=Divide(A=@n72, B=@n73)))
  n75 = ComponentMask(None=WorldPosition().XYZ) {b=true}
  n76 = Constant() {r=-1200}
  n77 = OneMinus(None=Saturate(None=Subtract(A=@n75, B=@n76)))
  NamedRerouteDeclaration(None=Subtract(A=Saturate(None=@n74), B=@n77)) {name=GM-Beach, nodeColor=<rgba 0.5 0 1 1>}
  n78 = Saturate(None=Subtract(A=@n75, B=Constant() {r=-700}))
  NamedRerouteDeclaration(None=Subtract(A=OneMinus(None=@n78), B=@n77)) {name=GM-Ocean, nodeColor=<rgba 0.5 0 1 1>}
  n79 = LinearInterpolate(A=Constant() {r=-0.5}, B=Constant() {r=1.5}, Alpha=LandscapeLayerSample() {desc=ForestFloor, parameterName=BiomeA})
  n80 = NamedRerouteDeclaration(None=MaterialFunctionCall(Transition Phase=Saturate(None=@n79), Height Texture=NamedRerouteUsage(), Contrast=Constant() {r=0.5}) {desc=ForestFloor Masked, materialFunction=&HeightLerp}.Alpha) {name=GM-ForestFloor-BeforeSnow, nodeColor=<rgba 0.5 0 1 1>}
  n81 = Constant() {r=6}
  n82 = NamedRerouteUsage()
  n83 = OneMinus(None=Saturate(None=Multiply(A=@n81, B=@n82)))
  NamedRerouteDeclaration(None=StaticSwitch(True=Multiply(A=@n80, B=@n83), False=@n80, Value=NamedRerouteUsage())) {name=GM-ForestFloor, nodeColor=<rgba 0.5 0 1 1>}
  NamedRerouteDeclaration(None=@n19.GrassHeightmap) {name=GrassHeightmap, nodeColor=<rgba 0.5 0 1 1>}
  NamedRerouteDeclaration(None=@n19.CommonLayerMaskForGrassFoliage) {name=GM-CommonLayers, nodeColor=<rgba 0.5 0 1 1>}
  n84 = StaticBoolParameter() {
   defaultValue=true
   desc=This should be false for the mountains MI where BomeB is heather instead of snow
   parameterName=Mountains-IsStaticSnow
  }
  n85 = LandscapeLayerSample() {desc=Reconnect this node after s22, parameterName=BiomeB}
  n86 = LandscapeLayerSample() {parameterName=BiomeD}
  n87 = StaticBool()
  n88 = RuntimeVirtualTextureSampleParameter() {
   group=RVT
   materialType=BaseColor_Normal_Specular_YCoCg
   parameterName=RVT Sample Target
   virtualTexture=/Game/Athena/Apollo/Maps/Landscape/VT/RVT_Apollo_Landscape
  }
  n89 = SetMaterialAttributes!(MaterialAttributes, Base Color=@n88.BaseColor, Specular=@n88.BaseColor, Roughness=@n88.BaseColor, Normal=@n88.BaseColor) {attributeSetTypes [
    #1
    #2
    #3
    #4
   ]}
  n90 = MaterialFunctionCall!(
   IsStaticSnow=NamedRerouteDeclaration(None=@n84) {name=Mountains-IsStaticSnow, nodeColor=<rgba 0.5 0 1 1>}
   UseAlternateStaticSnowMask=StaticBool() {value=true}
   AlternateStaticSnowMask=Saturate(None=Subtract(A=@n85, B=@n86))
   "UseRoad&&OcclusionMask"
   IsForGrassMeshToPushDown
   AdditionalMaskToRemoveSnow
   IsFallingSnowFX
   UseMinimapCoords
   IncreaseSnowUsingSlope
   UseSmoothBorder
   LerpHighM
   TemporalHeightLerpContrast
   IsLobbyIsland
   UseBR-IslandMask=@n87
   IsTree
   IsForRenderTarget
   UseSparkle
   UseSnowNormals=StaticBool() {value=true}
   UseAdditionalMaskToRemoveSnow
   IsTangentSpaceInput
   UseVertexNormalsForBasicSnow
   UseTangentSnowNormals=@n87
   Material=MaterialFunctionCall(NoRVT=@n18, RVT=@n89) {materialFunction=&FN_RVT_Switch_Material}.Result
  ) {materialFunction=/Game/Athena/Apollo/Environments/Landscape/Materials/Biome/UniqueBiomeMaterialFunctions/MF_Apollo_Snow_Progression}
  NamedRerouteDeclaration(None=@n90.SnowMaskForRVT) {name=SnowMaskForRVT, nodeColor=<rgba 0.5 0 1 1>}
  NamedRerouteDeclaration(None=@n90.SnowAlbedoForRVT) {name=SnowAlbedoForRVT, nodeColor=<rgba 0.5 0 1 1>}
  NamedRerouteDeclaration(None=StaticSwitch(True=@n90.SnowMask, False=Constant(), Value=NamedRerouteUsage())) {name=SnowMask, nodeColor=<rgba 0.5 0 1 1>}
  NamedRerouteDeclaration(None=@n28.SlopeScreeMask) {name=SlopeScreeMask, nodeColor=<rgba 0.5 0 1 1>}
  NamedRerouteDeclaration(None=@n19.GravelMaskForGrassFoliage) {name=GM-Gravel, nodeColor=<rgba 0.5 0 1 1>}
  NamedRerouteDeclaration(None=@n28.HeightmapForGrass) {name=HeightmapForGrass, nodeColor=<rgba 1 1 0 1>}
  NamedRerouteDeclaration(None=Constant()) {name=MF_ProceduralSnow-MatMask, nodeColor=<rgba 0.5 0 1 1>}
  NamedRerouteDeclaration(None=@n28.CliffMask) {name=CliffMask, nodeColor=<rgba 0.5 0 1 1>}
  NamedRerouteDeclaration(None=@n28.LandscapeGrassMeshMask) {name=GM-Cliff, nodeColor=<rgba 0.5 0 1 1>}
  NamedRerouteUsage()
  n91 = MaterialFunctionCall!(
   Heightmap=@n28.RegularHeightmap
   WorldnormalIn=@n29.Normals
   UseManualSnowMask=StaticBool() {value=true}
   SnowMask=LandscapeLayerSample() {parameterName=BiomeD}
  ) {materialFunction=/Game/Athena/Apollo/Environments/Landscape/Materials/Biome/UniqueBiomeMaterialFunctions/MF_ProceduralSnow}
  Reroute(None=@n91.MaterialMask)
  RuntimeVirtualTextureOutput!(
   BaseColor=MaterialFunctionCall!(BaseColor=@n35.Base Color) {materialFunction=/Game/Athena/Apollo/Environments/Landscape/Materials/Biome/CommonMaterials/MF_ReduceRTV-CompressionArtifacts}.BaseColor
   Specular=@n35.Specular
   Roughness=@n35.Specular
   Normal=@n35.Specular
   WorldHeight=ComponentMask(None=WorldPosition().XYZ) {b=true}
   Opacity
   Mask
   Displacement
   Mask4
  )
  ScalarParameter() {defaultValue=1.6832, parameterName=Threshold_Scree2}
  ScalarParameter() {defaultValue=-0.5, parameterName=BiomeC-GrassSpawnLa}
  ScalarParameter() {defaultValue=3, parameterName=BiomeC-GrassSpawnLb}
  ScalarParameter() {defaultValue=5, parameterName=BiomeC-GrassSpawnCont}
  ScalarParameter() {defaultValue=-0.5, parameterName=Biome-FFSpawnLa}
  ScalarParameter() {defaultValue=1.5, parameterName=Biome-FFSpawnLb}
  ScalarParameter() {defaultValue=0.5, parameterName=Biome-FFSpawnCont}
  ScalarParameter() {defaultValue=-4, parameterName=Threshold_Scree}
  n92 = StaticSwitchParameter(
   True=MaterialFunctionCall!(In=@n90.Material, Mask=LandscapeLayerSample() {parameterName=BiomeD}, UseCheapSlopeMask, MaskReflectionCurveOnZ, HueShiftOverride) {materialFunction=/Game/TechArt/Seasonal/LackLuster/MaterialFunctions/MF_LackLuster_Landscape_PostRVT}.Result
   False=@n90.Material
  ) {parameterName=UseLackluster}
  SetMaterialAttributes!(
   MaterialAttributes=MaterialFunctionCall!(
    Material=MaterialFunctionCall!(
     In=@n92
     UseHeightFalloff
     OnFireEmissiveOverride
     HeightFalloffRadius
     HeightFalloffHardness
     VariationMasksMobile
     In_Grass
     Is_Stylised_WE
    ) {desc=Fire, materialFunction=/Game/Packages/Fortress_SharedMaterials/Base_Material_Functions/MF_Curie_Landscape}.Result
    DebugInputAsBasecolor
    Non-RVT material=NamedRerouteUsage()
    DebugMaterialAs1Color=VectorParameter() {parameterName=DebugBiomeColor, defaultValue=<rgba 0.4069 0.2696 5 1>}.RGB
   ) {materialFunction=/Game/Athena/Apollo/Environments/Landscape/Materials/Biome/CommonMaterials/MF_LS_DEBUG}.Result
   Opacity Mask=LandscapeVisibilityMask()
  ) {attributeSetTypes [#5]}
  StaticSwitch(
   True=BlendMaterialAttributes(A=@n29.Result, B=@n91.Material, Alpha=@n91.Material)
   False=@n29.Result
   Value=NamedRerouteDeclaration(None=StaticBool()) {name=UseSnow, nodeColor=<rgba 0.5 0 1 1>}
  ) {desc=Use Procedural Snow}
  n93 = Constant() {r=2}
  n94 = NamedRerouteUsage()
  Subtract(A=@n61, B=Saturate(None=Multiply(A=@n93, B=@n94)))
 ]
}
