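# Terrain material function that tints landscape base color inside a world-projected region using the T_BeachCelebration_Macro texture, behind the UseEmissiveTerrainLayer static switch.
# World XY, offset and scaled by projection parameters, drives the texture UVs and a bounds mask that reads only X.
# Desaturated base color times texture and clamped pixel-normal Fresnel blends in by texture alpha and bounds.
# An EmissiveStrength blend exists but is unconnected, leaving the EmissiveColor input unwired.
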MaterialFunction /Game/Athena/Apollo/Environments/Landscape/Materials/Biome/UniqueBiomeMaterialFunctions/MF_EmissiveTerrain
graph {nodes [
  n1 = GetMaterialAttributes(None=FunctionInput() {inputName=MaterialAttributes, inputType=FunctionInput_MaterialAttributes}) {attributeGetTypes [#1]}
  n2 = ScalarParameter() {defaultValue=1, parameterName=ColorDesat}
  n3 = WorldPosition()
  n4 = VectorParameter() {parameterName=ProjectionPosition, defaultValue=<rgba -14109.6309 70000 -84 1>}
  n5 = ComponentMask(None=Divide(
    A=Subtract(A=@n3.XYZ, B=@n4.RGB)
    B=VectorParameter() {group=Terrain_Emissive, parameterName=ProjectionScale, defaultValue=<rgba 25000 25000 20000 1>}.RGB
   )) {g=true, r=true}
  n6 = TextureSample(UVs=Add(A=@n5) {constB=0.5}) {texture=/Game/Athena/Apollo/Environments/Landscape/Materials/Frontier/T_BeachCelebration_Macro}
  n7 = Fresnel(ExponentIn=ScalarParameter() {defaultValue=2, parameterName=FresnelExponent}, Normal=PixelNormalWS()) {exponent=2}
  n8 = ScalarParameter() {defaultValue=2, parameterName=FresnelPower}
  n9 = Multiply(A=Multiply(A=Desaturation(None=@n1.BaseColor, Fraction=@n2), B=@n6.RGB), B=Clamp(None=Power(Base=@n7, Exp=@n8)) {minDefault=0.25})
  n10 = MaterialFunctionCall(Float2=Round(None=Abs(None=@n5))) {materialFunction=&BreakOutFloat2Components}
  n11 = Max(A=@n10.R, B=@n10.R)
  n12 = Multiply(A=@n6.A, B=Clamp(None=OneMinus(None=@n11)))
  n13 = LinearInterpolate(A=@n1.BaseColor, B=Multiply(A=@n9, B=ScalarParameter() {defaultValue=2, parameterName=BaseColorLevels}), Alpha=@n12)
  FunctionOutput(None=StaticSwitchParameter(
    True=SetMaterialAttributes!(MaterialAttributes=@n1.MaterialAttributes, EmissiveColor, BaseColor=@n13) {attributeSetTypes [#2, #1]}
    False=@n1.MaterialAttributes
   ) {parameterName=UseEmissiveTerrainLayer})
  LinearInterpolate(B=Multiply(A=@n9, B=ScalarParameter() {defaultValue=1, group=Terrain_Emissive, parameterName=EmissiveStrength}), Alpha=@n12)
 ]}
