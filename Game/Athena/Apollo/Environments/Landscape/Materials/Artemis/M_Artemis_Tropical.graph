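# Tropical biome landscape material for Artemis terrain, also driving procedural foliage and physical materials from painted layer masks.
# Common layers blend with BiomeA jungle grass, BiomeB mud and BiomeC lava rock, then slope cliffs and BiomeD chrome, before writing to a runtime virtual texture.
# Final shading switches between RVT and simpler mobile layers, then adds snow, chrome, fire and debug functions.
# BiomeD is subtracted from several grass types, whose baked outputs require resaving levels after changes.
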Material /Game/Athena/Apollo/Environments/Landscape/Materials/Artemis/M_Artemis_Tropical
graph {
 properties {bTangentSpaceNormal=false, bUseMaterialAttributes=true, pixelDepthOffsetMode=PDOM_Legacy, usedShadingModels=Default Lit}
 nodes [
  n1 = NamedRerouteUsage()
  MaterialExpressionSubtract_0 = Subtract(
   A=MaterialFunctionCall!(In=@n1) {materialFunction=/Game/Athena/Apollo/Environments/Landscape/Materials/Biome/CommonMaterials/MF_LS_GrassSwitch}.Result
   B=LandscapeLayerSample() {parameterName=BiomeD}
  )
  MaterialExpressionSubtract_3 = Subtract(
   A=MaterialFunctionCall(In=@n1) {materialFunction=&MF_LS_GrassSwitch}.Result
   B=LandscapeLayerSample() {parameterName=BiomeD}
  )
  MaterialExpressionMaterialFunctionCall_17 = MaterialFunctionCall() {materialFunction=&MF_LS_GrassSwitch}
  MaterialExpressionSubtract_1 = Subtract(
   A=MaterialFunctionCall(In=NamedRerouteUsage()) {materialFunction=&MF_LS_GrassSwitch}.Result
   B=LandscapeLayerSample() {parameterName=BiomeD}
  )
  MaterialExpressionSubtract_4 = Subtract(
   A=MaterialFunctionCall() {materialFunction=&MF_LS_GrassSwitch}.Result
   B=LandscapeLayerSample() {parameterName=BiomeD}
  )
  MaterialExpressionSubtract_7 = Subtract(
   A=MaterialFunctionCall() {materialFunction=&MF_LS_GrassSwitch}.Result
   B=LandscapeLayerSample() {parameterName=BiomeD}
  )
  MaterialExpressionMaterialFunctionCall_22 = MaterialFunctionCall() {materialFunction=&MF_LS_GrassSwitch}
  MaterialExpressionMaterialFunctionCall_23 = MaterialFunctionCall(In=NamedRerouteUsage()) {materialFunction=&MF_LS_GrassSwitch}
  MaterialExpressionMaterialFunctionCall_24 = MaterialFunctionCall(In=NamedRerouteUsage()) {materialFunction=&MF_LS_GrassSwitch}
  MaterialExpressionMaterialFunctionCall_25 = MaterialFunctionCall() {materialFunction=&MF_LS_GrassSwitch}
  MaterialExpressionMaterialFunctionCall_26 = MaterialFunctionCall() {materialFunction=&MF_LS_GrassSwitch}
  MaterialExpressionSubtract_2 = Subtract(
   A=MaterialFunctionCall(In=NamedRerouteUsage()) {materialFunction=&MF_LS_GrassSwitch}.Result
   B=LandscapeLayerSample() {parameterName=BiomeD}
  )
  MaterialExpressionMaterialFunctionCall_28 = MaterialFunctionCall() {materialFunction=&MF_LS_GrassSwitch}
  LandscapeGrassOutput!(
   Grass_Standard=@MaterialExpressionSubtract_0
   Bush=@MaterialExpressionSubtract_3
   GrasslandFlowers=@MaterialExpressionMaterialFunctionCall_17.Result
   Scree=@MaterialExpressionSubtract_1
   Scree_Grass=@MaterialExpressionSubtract_4
   Heather=@MaterialExpressionSubtract_7
   ForestFloor=@MaterialExpressionMaterialFunctionCall_22.Result
   Beach=@MaterialExpressionMaterialFunctionCall_23.Result
   Ocean=@MaterialExpressionMaterialFunctionCall_24.Result
   DesertPlants=@MaterialExpressionMaterialFunctionCall_25.Result
   DesertScree=@MaterialExpressionMaterialFunctionCall_26.Result
   TropicalJungle=@MaterialExpressionSubtract_2
   FrontierPlants=@MaterialExpressionMaterialFunctionCall_28.Result
  ) {
   desc="This must have the exact inputs in every biome material!\r\nThis data is baked into the level data when the level is\r\nsaved, so any changes to this require the affected level\r\nto be saved && submitted."
   grassTypes [
    Grass_Standard {
     grassType=/Game/Athena/Apollo/Environments/Landscape/SmallFoliage/DataAssets/LandscapeGrassType/LandscapeGrassType_StandardGrass
     input {
      expression=&self:MaterialExpressionSubtract_0
      inputName=None
      mask=0
      maskA=0
      maskB=0
      maskG=0
      maskR=0
      outputIndex=0
     }
    }
    Bush {
     grassType=/Game/Athena/Environments/Landscape/Foliage/Athena_Foliage_Bushes
     input {
      expression=&self:MaterialExpressionSubtract_3
      inputName=None
      mask=0
      maskA=0
      maskB=0
      maskG=0
      maskR=0
      outputIndex=0
     }
    }
    GrasslandFlowers {
     grassType=/Game/Athena/Apollo/Environments/Landscape/SmallFoliage/DataAssets/LandscapeGrassType/LandscapeGrassType_FarmFlowers
     input {
      expression=&self:MaterialExpressionMaterialFunctionCall_17
      inputName=None
      mask=0
      maskA=0
      maskB=0
      maskG=0
      maskR=0
      outputIndex=0
     }
    }
    Scree {
     grassType=/Game/Athena/Apollo/Environments/Landscape/SmallFoliage/DataAssets/LandscapeGrassType/LandscapeGrassType_Scree
     input {
      expression=&self:MaterialExpressionSubtract_1
      inputName=None
      mask=0
      maskA=0
      maskB=0
      maskG=0
      maskR=0
      outputIndex=0
     }
    }
    Scree_Grass {
     grassType=/Game/Athena/Apollo/Environments/Landscape/SmallFoliage/DataAssets/LandscapeGrassType/LandscapeGrassType_ScreeGrass
     input {
      expression=&self:MaterialExpressionSubtract_4
      inputName=None
      mask=0
      maskA=0
      maskB=0
      maskG=0
      maskR=0
      outputIndex=0
     }
    }
    Heather {
     grassType=/Game/Athena/Apollo/Environments/Landscape/SmallFoliage/DataAssets/LandscapeGrassType/LandscapeGrassType_Heather
     input {
      expression=&self:MaterialExpressionSubtract_7
      inputName=None
      mask=0
      maskA=0
      maskB=0
      maskG=0
      maskR=0
      outputIndex=0
     }
    }
    ForestFloor {
     grassType=/Game/Athena/Apollo/Environments/Landscape/SmallFoliage/DataAssets/LandscapeGrassType/LandscapeGrassType_ForestGrass
     input {
      expression=&self:MaterialExpressionMaterialFunctionCall_22
      inputName=None
      mask=0
      maskA=0
      maskB=0
      maskG=0
      maskR=0
      outputIndex=0
     }
    }
    Beach {
     grassType=/Game/Athena/Apollo/Environments/Landscape/SmallFoliage/DataAssets/LandscapeGrassType/LandscapeGrassType_Beach
     input {
      expression=&self:MaterialExpressionMaterialFunctionCall_23
      inputName=None
      mask=0
      maskA=0
      maskB=0
      maskG=0
      maskR=0
      outputIndex=0
     }
    }
    Ocean {
     grassType=/Game/Athena/Apollo/Environments/Landscape/SmallFoliage/DataAssets/LandscapeGrassType/LandscapeGrassType_Ocean
     input {
      expression=&self:MaterialExpressionMaterialFunctionCall_24
      inputName=None
      mask=0
      maskA=0
      maskB=0
      maskG=0
      maskR=0
      outputIndex=0
     }
    }
    DesertPlants {
     grassType=/Game/Athena/Apollo/Environments/Landscape/SmallFoliage/DataAssets/LandscapeGrassType/LandscapeGrassType_DesertPlants
     input {
      expression=&self:MaterialExpressionMaterialFunctionCall_25
      inputName=None
      mask=0
      maskA=0
      maskB=0
      maskG=0
      maskR=0
      outputIndex=0
     }
    }
    DesertScree {
     grassType=/Game/Athena/Apollo/Environments/Landscape/SmallFoliage/DataAssets/LandscapeGrassType/LandscapeGrassType_DesertScree
     input {
      expression=&self:MaterialExpressionMaterialFunctionCall_26
      inputName=None
      mask=0
      maskA=0
      maskB=0
      maskG=0
      maskR=0
      outputIndex=0
     }
    }
    TropicalJungle {
     grassType=/Game/Athena/Apollo/Environments/Landscape/SmallFoliage/DataAssets/LandscapeGrassType/LandscapeGrassType_Nightmare
     input {
      expression=&self:MaterialExpressionSubtract_2
      inputName=None
      mask=0
      maskA=0
      maskB=0
      maskG=0
      maskR=0
      outputIndex=0
     }
    }
    FrontierPlants {
     grassType=/Game/Athena/Apollo/Environments/Landscape/SmallFoliage/DataAssets/LandscapeGrassType/NewLandscapeGrassType_FrontierGrass
     input {
      expression=&self:MaterialExpressionMaterialFunctionCall_28
      inputName=None
      mask=0
      maskA=0
      maskB=0
      maskG=0
      maskR=0
      outputIndex=0
     }
    }
   ]
  }
  MaterialExpressionMaterialFunctionCall_30 = MaterialFunctionCall!(SampleValue=LandscapeLayerSample() {parameterName=BiomeD}) {materialFunction=/Game/Athena/Apollo/Environments/Landscape/Materials/MaterialFunctions/MF_ArtemisChromePhysicalMaterialMask}
  n2 = NamedRerouteUsage()
  n3 = OneMinus(None=@n2) {desc="White=Cliffs"}
  n4 = Max(
   A=Max(A=LandscapeLayerSample() {parameterName=BiomeA}, B=LandscapeLayerSample() {parameterName=BiomeB})
   B=LandscapeLayerSample() {parameterName=BiomeC}
  )
  n5 = Max(A=LandscapeLayerSample() {parameterName=Gravel}, B=LandscapeLayerSample() {parameterName=Mud})
  n6 = Constant() {r=2}
  n7 = Max(A=@n4, B=Saturate(None=Multiply(A=@n5, B=@n6)))
  n8 = Subtract(A=@n7, B=@n3) {desc="Remove Cliff from Gravel,Mud, Sand, Snow"}
  n9 = LandscapeLayerSample() {parameterName=Sand}
  n10 = Saturate(None=Multiply(A=LandscapeLayerSample() {parameterName=Asphalt}, B=Constant() {r=3}))
  MaterialExpressionSaturate_12 = Saturate(None=Subtract(A=Saturate(None=@n8), B=Max(A=@n9, B=@n10)) {desc="Remove Asphalt, Sand"}) {desc=Final Dirt mask}
  MaterialExpressionReroute_51 = Reroute(None=@n3) {desc=Final Stone mask}
  n11 = Subtract(A=@n2, B=Max(A=@n9, B=@n7)) {desc="Remove Gravel,Mud, Sand, Snow"}
  MaterialExpressionSaturate_14 = Saturate(None=Subtract(A=Saturate(None=@n11), B=@n10) {desc=Remove Asphalt}) {desc=Final Grass mask}
  MaterialExpressionReroute_61 = Reroute(None=@n10)
  n12 = Saturate(None=Subtract(A=@n9, B=@n3) {desc=Remove Cliff})
  n13 = Max(A=@n4, B=@n10)
  MaterialExpressionReroute_42 = Reroute(None=Saturate(None=Subtract(A=@n12, B=@n13)))
  LandscapePhysicalMaterialOutput!(
   Chrome_PhysicalMaterial=@MaterialExpressionMaterialFunctionCall_30.Result
   Dirt=@MaterialExpressionSaturate_12
   Stone=@MaterialExpressionReroute_51
   Grass=@MaterialExpressionSaturate_14
   Asphalt=@MaterialExpressionReroute_61
   Sand=@MaterialExpressionReroute_42
  ) {inputs [
    {
     physicalMaterial=/Game/TechArt/Seasonal/LackLuster/Physics/Chrome_PhysicalMaterial
     input {
      expression=&self:MaterialExpressionMaterialFunctionCall_30
      inputName=None
      mask=0
      maskA=0
      maskB=0
      maskG=0
      maskR=0
      outputIndex=0
     }
    }
    {
     physicalMaterial=/Game/FortressPhysicalMaterials/Dirt
     input {
      expression=&self:MaterialExpressionSaturate_12
      inputName=None
      mask=0
      maskA=0
      maskB=0
      maskG=0
      maskR=0
      outputIndex=0
     }
    }
    {
     physicalMaterial=/Game/FortressPhysicalMaterials/Stone
     input {
      expression=&self:MaterialExpressionReroute_51
      inputName=None
      mask=0
      maskA=0
      maskB=0
      maskG=0
      maskR=0
      outputIndex=0
     }
    }
    {
     physicalMaterial=/Game/FortressPhysicalMaterials/Grass
     input {
      expression=&self:MaterialExpressionSaturate_14
      inputName=None
      mask=0
      maskA=0
      maskB=0
      maskG=0
      maskR=0
      outputIndex=0
     }
    }
    {
     physicalMaterial=/Game/FortressPhysicalMaterials/Asphalt
     input {
      expression=&self:MaterialExpressionReroute_61
      inputName=None
      mask=0
      maskA=0
      maskB=0
      maskG=0
      maskR=0
      outputIndex=0
     }
    }
    {
     physicalMaterial=/Game/FortressPhysicalMaterials/Sand
     input {
      expression=&self:MaterialExpressionReroute_42
      inputName=None
      mask=0
      maskA=0
      maskB=0
      maskG=0
      maskR=0
      outputIndex=0
     }
    }
   ]}
  n14 = LinearInterpolate(A=@MaterialExpressionSaturate_12, B=@MaterialExpressionReroute_51, Alpha=ScalarParameter() {parameterName=DEBUG-PhysMat1})
  n15 = ScalarParameter() {parameterName=DEBUG-PhysMat2}
  LinearInterpolate(
   A=LinearInterpolate(
    A=LinearInterpolate(A=@n14, B=@MaterialExpressionSaturate_14, Alpha=@n15)
    B=@MaterialExpressionReroute_61
    Alpha=ScalarParameter() {parameterName=DEBUG-PhysMat3}
   )
   B=@MaterialExpressionReroute_42
   Alpha=ScalarParameter() {defaultValue=1, parameterName=DEBUG-PhysMat4}
  )
  n16 = Constant() {r=1.25}
  n17 = TextureObject() {texture=/Game/Environments/Artemis/Sets/Terrain/Textures/T_Artemis_Lava_Rock_03_D}
  n18 = TextureObject() {samplerType=SAMPLERTYPE_Normal, texture=/Game/Environments/Artemis/Sets/Terrain/Textures/T_Artemis_Lava_Rock_03_N}
  n19 = TextureObject() {samplerType=SAMPLERTYPE_Masks, texture=/Game/Environments/Artemis/Sets/Terrain/Textures/T_Artemis_Lava_Rock_03_S}
  MaterialFunctionCall!(Scale=@n16, Diffuse=@n17, Normal=@n18, S_Mask=@n19) {materialFunction=/Game/Athena/Apollo/Environments/Landscape/Materials/Biome/UniqueBiomeMaterialFunctions/MF_Biome_Generic}
  n20 = MaterialFunctionCall!(AdditionalGrassBorderColorMask, PutSandBeforeGravel=StaticBool() {value=true}, UseOceanLevel) {materialFunction=/Game/Athena/Apollo/Environments/Landscape/Materials/Biome/CommonMaterials/MF_CommonLandscapeLayers}
  NamedRerouteDeclaration(None=@n20.CommonLayerMaskForGrassFoliage) {name=GM-CommonLayers, nodeColor=<rgba 0.5 0 1 1>}
  n21 = LandscapeLayerSample() {parameterName=BiomeB}
  n22 = LandscapeLayerSample() {parameterName=BiomeC}
  NamedRerouteDeclaration(None=Subtract(A=LandscapeLayerSample() {parameterName=BiomeA}, B=Max(A=@n21, B=@n22))) {name=GM-TropicalJungle, nodeColor=<rgba 0.5 0 1 1>}
  n23 = Max(A=NamedRerouteUsage(), B=NamedRerouteUsage())
  n24 = LandscapeLayerSample() {desc=ForestFloor, parameterName=BiomeA}
  n25 = Max(A=LandscapeLayerSample() {parameterName=BiomeB}, B=LandscapeLayerSample() {parameterName=BiomeC})
  n26 = Saturate(None=Multiply(A=Max(A=@n24, B=@n25), B=Constant() {r=3}))
  NamedRerouteDeclaration(None=OneMinus(None=Max(A=@n23, B=@n26))) {name=GM-Grass, nodeColor=<rgba 0.5 0 1 1>}
  n27 = LandscapeLayerSample() {parameterName=Sand}
  n28 = Max(A=NamedRerouteUsage(), B=NamedRerouteUsage())
  n29 = ComponentMask(None=WorldPosition().XYZ) {b=true}
  n30 = Constant() {r=200}
  n31 = Subtract(A=Saturate(None=Subtract(A=@n27, B=@n28)), B=Saturate(None=Divide(A=@n29, B=@n30)))
  n32 = ComponentMask(None=WorldPosition().XYZ) {b=true}
  n33 = Constant() {r=-1200}
  n34 = OneMinus(None=Saturate(None=Subtract(A=@n32, B=@n33)))
  NamedRerouteDeclaration(None=Subtract(A=Saturate(None=@n31), B=@n34)) {name=GM-Beach, nodeColor=<rgba 0.5 0 1 1>}
  n35 = Saturate(None=Subtract(A=@n32, B=Constant() {r=-700}))
  NamedRerouteDeclaration(None=Subtract(A=OneMinus(None=@n35), B=@n34)) {name=GM-Ocean, nodeColor=<rgba 0.5 0 1 1>}
  n36 = MaterialFunctionCall(
   Scale=Constant() {r=1.5}
   Diffuse=TextureObject() {texture=/Game/Environments/Apollo/Sets/Nightmare/Textures/T_Nightmare_JungleGrass_D_01}
   Normal=TextureObject() {samplerType=SAMPLERTYPE_Normal, texture=/Game/Environments/Apollo/Sets/Nightmare/Textures/T_Nightmare_JungleGrass_N_01}
   S_Mask=TextureObject() {samplerType=SAMPLERTYPE_Masks, texture=/Game/Environments/Apollo/Sets/Nightmare/Textures/T_Nightmare_JungleGrass_S_01}
  ) {materialFunction=&MF_Biome_Generic}
  n37 = MaterialFunctionCall!(
   Material 1=@n20.Material
   Height Texture 1=@n20.Material
   Material 2=@n36.Material
   Height Texture 2=NamedRerouteDeclaration(None=@n36.Height) {name=BiomeA-ForestFloorHeight, nodeColor=<rgba 0.5 0 1 1>}
   Transistion Phase=LandscapeLayerSample() {parameterName=BiomeA}
   UseContrast
   Contrast
   AddLayerMask
  ) {materialFunction=/Game/Athena/Apollo/Environments/Landscape/Materials/Biome/CommonMaterials/MF_BlendLandscapeLayers}
  n38 = MaterialFunctionCall(
   Scale=Constant() {r=1.5}
   Diffuse=TextureObject() {texture=/Game/Environments/Apollo/Sets/Nightmare/Textures/T_Nightmare_Mud_D_01}
   Normal=TextureObject() {samplerType=SAMPLERTYPE_Normal, texture=/Game/Environments/Apollo/Sets/Nightmare/Textures/T_Nightmare_Mud_N_01}
   S_Mask=TextureObject() {samplerType=SAMPLERTYPE_Masks, texture=/Game/Environments/Apollo/Sets/Nightmare/Textures/T_Nightmare_Mud_S_01}
  ) {materialFunction=&MF_Biome_Generic}
  n39 = MaterialFunctionCall(
   Material 1=@n37.BlendedMaterial
   Height Texture 1=@n37.BlendedMaterial
   Material 2=@n38.Material
   Height Texture 2=@n38.Material
   Transistion Phase=LandscapeLayerSample() {parameterName=BiomeB}
  ) {materialFunction=&MF_BlendLandscapeLayers}
  n40 = Multiply(A=@n16, B=LandscapeLayerCoords() {mappingScale=10})
  n41 = TextureSample(UVs=@n40, Tex=@n17) {samplerSource=SSM_Wrap_WorldGroupSettings}
  n42 = TextureSample(UVs=@n40, Tex=@n19) {samplerSource=SSM_Wrap_WorldGroupSettings, samplerType=SAMPLERTYPE_Masks}
  n43 = SetMaterialAttributes!(
   MaterialAttributes
   Base Color=Multiply(A=@n41.RGB, B=Constant() {r=0.5})
   Normal=TextureSample(UVs=@n40, Tex=@n18) {samplerSource=SSM_Wrap_WorldGroupSettings, samplerType=SAMPLERTYPE_Normal}.RGB
   Specular=@n42.R
   Roughness=@n42.R
  ) {attributeSetTypes [
    #1
    #2
    #3
    #4
   ]}
  n44 = MaterialFunctionCall(
   Material 1=@n39.BlendedMaterial
   Height Texture 1=@n39.BlendedMaterial
   Material 2=@n43
   Height Texture 2=@n41.A
   Transistion Phase=LandscapeLayerSample() {parameterName=BiomeC}
   UseContrast=StaticBool() {value=true}
   Contrast=Constant() {r=0.5}
  ) {materialFunction=&MF_BlendLandscapeLayers}
  n45 = MaterialFunctionCall!(
   Material=@n44.BlendedMaterial
   Heightmap=@n44.BlendedMaterial
   NewSlopeMaskMask
   Displacement
   AsphaltLayer=LandscapeLayerSample() {parameterName=Sand}
   UseAlternateCliffMat
   AltCliffHeight
   AltCliffMat
  ) {materialFunction=/Game/Athena/Apollo/Environments/Landscape/Materials/Biome/CommonMaterials/MF_SlopeBlendCliff}
  NamedRerouteDeclaration(None=@n45.SlopeScreeMask) {name=SlopeScreeMask, nodeColor=<rgba 0.5 0 1 1>}
  NamedRerouteDeclaration(None=@n20.GravelMaskForGrassFoliage) {name=GM-Gravel, nodeColor=<rgba 0.5 0 1 1>}
  NamedRerouteDeclaration(None=@n45.HeightmapForGrass) {name=HeightmapForGrass, nodeColor=<rgba 1 1 0 1>}
  n46 = LinearInterpolate(A=Constant() {r=-0.5}, B=Constant() {r=1.5}, Alpha=LandscapeLayerSample() {desc=ForestFloor, parameterName=BiomeA})
  NamedRerouteDeclaration(None=MaterialFunctionCall(Transition Phase=Saturate(None=@n46), Height Texture=NamedRerouteUsage(), Contrast=Constant() {r=0.5}) {desc=ForestFloor, materialFunction=&HeightLerp}.Alpha) {name=GM-ForestFloor, nodeColor=<rgba 0.5 0 1 1>}
  NamedRerouteDeclaration(None=@n45.CliffMask) {name=CliffMask, nodeColor=<rgba 0.5 0 1 1>}
  NamedRerouteDeclaration(None=@n45.LandscapeGrassMeshMask) {name=GM-Cliff, nodeColor=<rgba 0.5 0 1 1>}
  n47 = Saturate(None=Multiply(A=LandscapeLayerSample() {parameterName=Sand}, B=Constant() {r=4}))
  n48 = LandscapeLayerSample() {parameterName=Mud}
  n49 = LandscapeLayerSample() {parameterName=Asphalt}
  n50 = Max(A=Max(A=Max(A=@n47, B=@n48), B=Ceil(None=@n49)), B=NamedRerouteUsage())
  n51 = NamedRerouteUsage()
  n52 = LandscapeLayerSample() {parameterName=BiomeB}
  n53 = LandscapeLayerSample() {parameterName=BiomeC}
  n54 = Multiply(A=Max(A=LandscapeLayerSample() {desc=ForestFloor, parameterName=BiomeA}, B=Max(A=@n52, B=@n53)), B=Constant() {r=3})
  OneMinus(None=Max(A=Max(A=@n50, B=@n51), B=Saturate(None=@n54)))
  n55 = GetMaterialAttributes(None=MaterialFunctionCall!(PutSandBeforeGravelOverride, UseSnow=StaticBool(), UseOceanLevel) {materialFunction=/Game/Athena/Apollo/Environments/Landscape/Materials/Biome/CommonMaterials/Mobile/MF_Mobile_CommonLandscapeLayers}.Material) {attributeGetTypes [#1]}
  MaterialExpressionConstant_2 = Constant()
  n56 = LandscapeLayerBlend!(
   Layer Gravel=@MaterialExpressionConstant_2
   Layer Mud=@MaterialExpressionConstant_2
   Layer Sand=@MaterialExpressionConstant_2
   Layer Asphalt=@MaterialExpressionConstant_2
  ) {layers [
    {
     blendType=LB_WeightBlend
     constHeightInput=0
     layerName=Gravel
     previewWeight=0
     constLayerInput=<0 0 0>
     heightInput {expression=null, inputName=None, mask=0, maskA=0, maskB=0, maskG=0, maskR=0, outputIndex=0}
     layerInput {
      expression=&self:MaterialExpressionConstant_2
      inputName=None
      mask=0
      maskA=0
      maskB=0
      maskG=0
      maskR=0
      outputIndex=0
     }
    }
    {
     blendType=LB_WeightBlend
     constHeightInput=0
     layerName=Mud
     previewWeight=0
     constLayerInput=<0 0 0>
     heightInput {expression=null, inputName=None, mask=0, maskA=0, maskB=0, maskG=0, maskR=0, outputIndex=0}
     layerInput {
      expression=&self:MaterialExpressionConstant_2
      inputName=None
      mask=0
      maskA=0
      maskB=0
      maskG=0
      maskR=0
      outputIndex=0
     }
    }
    {
     blendType=LB_WeightBlend
     constHeightInput=0
     layerName=Sand
     previewWeight=0
     constLayerInput=<0 0 0>
     heightInput {expression=null, inputName=None, mask=0, maskA=0, maskB=0, maskG=0, maskR=0, outputIndex=0}
     layerInput {
      expression=&self:MaterialExpressionConstant_2
      inputName=None
      mask=0
      maskA=0
      maskB=0
      maskG=0
      maskR=0
      outputIndex=0
     }
    }
    {
     blendType=LB_WeightBlend
     constHeightInput=0
     layerName=Asphalt
     previewWeight=0
     constLayerInput=<0 0 0>
     heightInput {expression=null, inputName=None, mask=0, maskA=0, maskB=0, maskG=0, maskR=0, outputIndex=0}
     layerInput {
      expression=&self:MaterialExpressionConstant_2
      inputName=None
      mask=0
      maskA=0
      maskB=0
      maskG=0
      maskR=0
      outputIndex=0
     }
    }
   ]}
  n57 = Multiply(A=@n56) {constB=0}
  n58 = MaterialFunctionCall!(MaterialAttributes=SetMaterialAttributes!(MaterialAttributes=@n55.MaterialAttributes, Base Color=Add(A=@n55.Base Color, B=@n57)) {desc="This must remain in material,\r\nnot a function.", attributeSetTypes [#1]}) {materialFunction=/Game/Athena/Apollo/Environments/Landscape/Materials/Biome/UniqueBiomeMaterialFunctions/MF_EmissiveTerrain}
  n59 = MaterialFunctionCall!(Material=MaterialFunctionCall!(MA=@n45.Material, Mask=LandscapeLayerSample() {parameterName=BiomeD}, HeightIn=@n45.Material, SlopeMask=@n45.Material, BlendLandscapeLayers) {materialFunction=/Game/TechArt/Seasonal/LackLuster/MaterialFunctions/MF_LackLuster_Landscape_RVT}.Result) {materialFunction=/Game/Athena/Apollo/Maps/Landscape/VT/MaterialFunctions/TransformTangentToWorld}
  n60 = GetMaterialAttributes(None=MaterialFunctionCall!(MobileMaterial=@n58.Result, HighEndMaterial=NamedRerouteDeclaration(None=@n59.Result) {name=MaterialForRVT, nodeColor=<rgba 0.5 0 1 1>}) {materialFunction=/Game/Athena/Asteria/Environments/Landscape/Materials/MaterialFunctions/FN_RvtWriteSwitch}.Result) {attributeGetTypes [
    #1
    #3
    #4
    #2
   ]}
  RuntimeVirtualTextureOutput!(
   BaseColor=MaterialFunctionCall!(BaseColor=@n60.Base Color) {materialFunction=/Game/Athena/Apollo/Environments/Landscape/Materials/Biome/CommonMaterials/MF_ReduceRTV-CompressionArtifacts}.BaseColor
   Specular=@n60.Specular
   Roughness=@n60.Specular
   Normal=@n60.Specular
   WorldHeight=ComponentMask(None=WorldPosition().XYZ) {b=true}
   Opacity
   Mask
   Displacement
   Mask4
  )
  ScalarParameter() {defaultValue=1.25, parameterName=BiomeC-Scale}
  ScalarParameter() {defaultValue=0.5, parameterName=LavaRockDiffuseMod}
  ScalarParameter() {defaultValue=0.5, parameterName=BiomeC-cont}
  ScalarParameter() {defaultValue=1.5, parameterName=BiomeA-Scale}
  ScalarParameter() {defaultValue=1.5, parameterName=BiomeB-Scale}
  n61 = StaticBool()
  n62 = RuntimeVirtualTextureSampleParameter() {
   group=RVT
   materialType=BaseColor_Normal_Specular_YCoCg
   parameterName=RVT Sample Target
   virtualTexture=/Game/Athena/Apollo/Maps/Landscape/VT/RVT_Apollo_Landscape
  }
  n63 = SetMaterialAttributes!(MaterialAttributes, Base Color=@n62.BaseColor, Specular=@n62.BaseColor, Roughness=@n62.BaseColor, Normal=@n62.BaseColor) {attributeSetTypes [
    #1
    #3
    #4
    #2
   ]}
  n64 = MaterialFunctionCall(NoRVT=@n58.Result, RVT=@n63) {materialFunction=&FN_RVT_Switch_Material}
  n65 = MaterialFunctionCall!(
   In=MaterialFunctionCall!(
    IsStaticSnow
    UseAlternateStaticSnowMask
    AlternateStaticSnowMask
    "UseRoad&&OcclusionMask"
    IsForGrassMeshToPushDown
    AdditionalMaskToRemoveSnow
    IsFallingSnowFX
    UseMinimapCoords
    IncreaseSnowUsingSlope
    UseSmoothBorder
    LerpHighM
    TemporalHeightLerpContrast
    IsLobbyIsland
    UseBR-IslandMask=@n61
    IsTree
    IsForRenderTarget
    UseSparkle
    UseSnowNormals=StaticBool() {value=true}
    UseAdditionalMaskToRemoveSnow
    IsTangentSpaceInput
    UseVertexNormalsForBasicSnow
    UseTangentSnowNormals=@n61
    Material=MaterialProxyReplace(Realtime=@n64.Result, MaterialProxy=@n58.Result) {desc=This is not in any other landscape materials?}
   ) {materialFunction=/Game/Athena/Apollo/Environments/Landscape/Materials/Biome/UniqueBiomeMaterialFunctions/MF_Apollo_Snow_Progression}.Material
   Mask=LandscapeLayerSample() {parameterName=BiomeD}
   UseCheapSlopeMask
   MaskReflectionCurveOnZ
   HueShiftOverride
  ) {materialFunction=/Game/TechArt/Seasonal/LackLuster/MaterialFunctions/MF_LackLuster_Landscape_PostRVT}
  SetMaterialAttributes!(
   MaterialAttributes=MaterialFunctionCall!(
    Material=MaterialFunctionCall!(
     In=@n65.Result
     UseHeightFalloff
     OnFireEmissiveOverride
     HeightFalloffRadius
     HeightFalloffHardness
     VariationMasksMobile
     In_Grass
     Is_Stylised_WE
    ) {desc=Fire, materialFunction=/Game/Packages/Fortress_SharedMaterials/Base_Material_Functions/MF_Curie_Landscape}.Result
    DebugInputAsBasecolor
    Non-RVT material=NamedRerouteUsage()
    DebugMaterialAs1Color=VectorParameter() {parameterName=DebugBiomeColor, defaultValue=<rgba 0 0.5 0 1>}.RGB
   ) {materialFunction=/Game/Athena/Apollo/Environments/Landscape/Materials/Biome/CommonMaterials/MF_LS_DEBUG}.Result
   Opacity Mask=LandscapeVisibilityMask()
  ) {attributeSetTypes [#5]}
 ]
}
